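# Material function for the BitFight character body that adds animated arrow and random-segment glows to incoming material attributes.
# Time-seeded random values pick cells of a two-column, three-row UV1 grid for the arrow mask, limited by the Mask input and a UV0 step, alongside a blinking toggle.
# The EM texture's G, B and A channels decode into segment IDs, matched against two random timers and masked by R.
# The resulting masks drive emissive color and darken base color.
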MaterialFunction /BRCosmetics/Characters/Player/Female/Medium/Bodies/F_MED_BitFight/Materials/MF_BitFight_ArrowsSequence
graph {nodes [
  n1 = LinearInterpolate(
   A=NamedRerouteUsage()
   B=VectorParameter() {parameterName=Active Arrow Color, defaultValue=<rgba 10 0.0037 1.6698 1>}.RGB
   Alpha=NamedRerouteUsage()
  )
  n2 = LinearInterpolate(A=NamedRerouteUsage(), B=Multiply(A=NamedRerouteUsage()) {constB=5}, Alpha=NamedRerouteUsage())
  n3 = NamedRerouteUsage()
  n4 = Add(A=NamedRerouteUsage(), B=NamedRerouteUsage())
  FunctionOutput(None=SetMaterialAttributes!(MaterialAttributes=NamedRerouteUsage(), EmissiveColor=Add(A=@n1, B=@n2), BaseColor=LinearInterpolate(A=@n3, Alpha=@n4) {constB=0}) {attributeSetTypes [#1, #2]}) {outputName=OutMat}
  n5 = Time()
  n6 = Multiply(A=@n5) {constB=6}
  n7 = Floor(None=MaterialFunctionCall!(Input=Floor(None=@n6), min=Constant(), max=Constant() {r=5.99}, seed=Constant() {r=7}) {materialFunction=/InterchangeAssets/Functions/MX_RandomFloat}.Result)
  n8 = Constant() {r=2}
  NamedRerouteDeclaration(None=Floor(None=Fmod(A=@n7, B=@n8))) {name=ColumnID, nodeColor=<rgba 0 0.6016 1 1>}
  NamedRerouteDeclaration(None=Floor(None=Divide(A=@n7, B=@n8))) {name=RowID, nodeColor=<rgba 0.9219 1 0 1>}
  n9 = GetMaterialAttributes(None=FunctionInput() {inputName=InMat, inputType=FunctionInput_MaterialAttributes}) {attributeGetTypes [#1, #2]}
  NamedRerouteDeclaration(None=@n9.EmissiveColor) {name=InEmissive, nodeColor=<rgba 1 0.4219 0 1>}
  NamedRerouteDeclaration(None=@n9.MaterialAttributes) {name=InMat, nodeColor=<rgba 0.1016 0 1 1>}
  n10 = MaterialFunctionCall(Float2=TextureCoordinate() {coordinateIndex=1}) {materialFunction=&BreakOutFloat2Components}
  n11 = Floor(None=Multiply(A=@n10.R) {constB=2})
  n12 = Multiply(A=@n10.G) {constB=3}
  n13 = Abs(None=Subtract(A=Floor(None=@n12), B=NamedRerouteUsage()))
  n14 = Multiply(A=LinearInterpolate(A=OneMinus(None=@n11), B=@n11, Alpha=NamedRerouteUsage()), B=OneMinus(None=Saturate(None=@n13)))
  n15 = TextureCoordinate()
  n16 = Multiply(A=FunctionInput() {inputName=Mask, inputType=FunctionInput_Scalar}, B=Step(X=ComponentMask(None=@n15) {r=true}) {constX=0.5, constY=0.5})
  NamedRerouteDeclaration(None=Multiply(A=Multiply(A=@n14, B=@n16), B=NamedRerouteUsage())) {name=Arrow Mask, nodeColor=<rgba 0.2187 1 0 1>}
  n17 = Multiply(A=@n5) {constB=2}
  NamedRerouteDeclaration(None=Fmod(A=Floor(None=@n17), B=Constant() {r=2})) {name=GlowActive, nodeColor=<rgba 0.7812 1 0 1>}
  NamedRerouteDeclaration(None=@n9.BaseColor) {name=InBaseColor, nodeColor=<rgba 0.9453 1 0 1>}
  n18 = TextureSample() {samplerType=SAMPLERTYPE_Masks, texture=/BRCosmetics/Characters/Player/Female/Medium/Bodies/F_MED_BitFight/Textures/T_BitFight_Body_EM}
  n19 = Floor(None=Multiply(A=Time(), B=ScalarParameter() {defaultValue=3, parameterName=RandomSpeed}))
  n20 = Constant() {r=1}
  n21 = Constant() {r=12.99}
  n22 = Constant() {r=5}
  n23 = Subtract(
   A=NamedRerouteUsage()
   B=Floor(None=MaterialFunctionCall(Input=@n19, min=@n20, max=@n21, seed=@n22) {materialFunction=&MX_RandomFloat}.Result)
  )
  n24 = OneMinus(None=Saturate(None=Abs(None=@n23)))
  n25 = Time()
  n26 = ScalarParameter() {defaultValue=2, parameterName=RandomSpeed}
  n27 = Floor(None=Add(A=Multiply(A=@n25, B=@n26)) {constB=0.5})
  n28 = Constant() {r=1}
  n29 = Constant() {r=12.99}
  n30 = Constant() {r=7}
  n31 = Subtract(
   A=NamedRerouteUsage()
   B=Floor(None=MaterialFunctionCall(Input=@n27, min=@n28, max=@n29, seed=@n30) {materialFunction=&MX_RandomFloat}.Result)
  )
  n32 = OneMinus(None=Saturate(None=Abs(None=@n31)))
  NamedRerouteDeclaration(None=Multiply(A=@n18.R, B=Max(A=@n24, B=@n32))) {name=Random Emissive, nodeColor=<rgba 1 0.3047 0 1>}
  n33 = Round(None=Multiply(A=@n18.G) {constB=4})
  n34 = Multiply(A=@n18.B) {constB=4}
  n35 = Multiply(A=Add(A=Round(None=@n34)) {constB=4}, B=Ceil(None=@n18.B))
  n36 = Add(A=Round(None=Multiply(A=@n18.A) {constB=4})) {constB=8}
  n37 = Ceil(None=@n18.A)
  NamedRerouteDeclaration(None=Add(A=Add(A=@n33, B=@n35), B=Multiply(A=@n36, B=@n37))) {name=IDs, nodeColor=<rgba 1 0 0.5391 1>}
 ]}
