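# Control Rig for the F_MED_BitFight female body cosmetic, posing its jacket, belt, game-controller prop and leg cube bones procedurally.
# After the shared DeformRig function, it parent-constrains prop bones to body bones and drives the controller from left-thigh swing.
# Spherical pose readers on that thigh set leg variables that weight corrective offsets.
# Jacket and leg cube fixes also respond to targeting, crouch height, spine rotation and knee attributes, often smoothed through curves and interpolation.
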ControlRigBlueprint /BRCosmetics/Characters/Player/Female/Medium/Bodies/F_MED_BitFight/Meshes/F_MED_BitFight_CtrlRig
graphs [
 Multiply_2_ContainedGraph {nodes [
   n1 = RigVMFunctionEntryNode()
   RigVMFunctionReturnNode(Result=&RigVMFunction_MathDoubleMul(B=@n1.C, A=&RigVMFunction_MathDoubleMul(A=@n1.A, B=@n1.B).Result).Result)
  ]}
 New Function_ContainedGraph {nodes [
   RigVMFunctionEntryNode()
   > &RigUnit_ParentConstraint(
    AdvancedSettings=(RotationOrderForFilter=XZY)
    Child=(Type=Bone,Name="for_aim_belt_front")
    Parents="((Item=(Type=Bone,Name=\"dyn_jacket_r_2\"),Weight=0.0001),(Item=(Type=Bone,Name=\"thigh_l\"),Weight=0.0003),(Item=(Type=Bone,Name=\"dyn_controller\"),Weight=0.0004),(Item=(Type=Bone,Name=\"dyn_jacket_b\"),Weight=0.0006),(Item=(Type=Bone,Name=\"thigh_r\"),Weight=0.00 ...(truncated)"
    Weight=1
   )
   > &RigUnit_ParentConstraint(
    AdvancedSettings=(RotationOrderForFilter=XZY)
    Child=(Type=Bone,Name="base_main_chain")
    Parents="((Item=(Type=Bone,Name=\"thigh_l\"),Weight=0.0002),(Item=(Type=Bone,Name=\"deform_glute_r\"),Weight=0.0086),(Item=(Type=Bone,Name=\"dyn_controller\"),Weight=0.0266),(Item=(Type=Bone,Name=\"deform_glute_l\"),Weight=0.356),(Item=(Type=Bone,Name=\"pelvis\"),Weight=0.60 ...(truncated)"
    Weight=1
   )
   > &RigUnit_ParentConstraint(
    AdvancedSettings=(RotationOrderForFilter=XZY)
    Child=(Type=Bone,Name="base_cube_leg_l")
    Parents="((Item=(Type=Bone,Name=\"calf_twist_02_l\"),Weight=0.0004),(Item=(Type=Bone,Name=\"deform_knee_out_l\"),Weight=0.0012),(Item=(Type=Bone,Name=\"calf_l\"),Weight=0.0163),(Item=(Type=Bone,Name=\"deform_knee_in_l\"),Weight=0.0559),(Item=(Type=Bone,Name=\"thigh_twist_01 ...(truncated)"
    Weight=1
   )
   > &RigUnit_ParentConstraint(
    AdvancedSettings=(RotationOrderForFilter=XZY)
    Child=(Type=Bone,Name="base_cube_leg_r")
    Parents="((Item=(Type=Bone,Name=\"calf_twist_02_r\"),Weight=0.0005),(Item=(Type=Bone,Name=\"deform_knee_out_r\"),Weight=0.0013),(Item=(Type=Bone,Name=\"calf_r\"),Weight=0.0175),(Item=(Type=Bone,Name=\"deform_knee_in_r\"),Weight=0.0623),(Item=(Type=Bone,Name=\"thigh_twist_01 ...(truncated)"
    Weight=1
   )
   > RigVMFunctionReturnNode()
  ]}
 New Function_ContainedGraph {nodes [
   RigVMFunctionEntryNode()
   > &RigUnit_RotationConstraintLocalSpaceOffset(
    AdvancedSettings=(InterpolationType=Average,RotationOrderForFilter=XZY)
    Child=(Type=Bone,Name="thigh_l_constraint")
    Filter=(bX=True,bY=True,bZ=True)
    Parents=((Item=(Type=Bone,Name="thigh_l"),Weight=1))
    Weight=1
   )
   > /Script/ControlRig.RigUnit_PositionConstraintLocalSpaceOffset(
    Child=(Type=Bone,Name="thigh_l_constraint")
    Filter=(bX=True,bY=True,bZ=True)
    Parents=((Item=(Type=Bone,Name="thigh_l"),Weight=1))
    Weight=1
    bMaintainOffset=True
   ) {is=struct RigVMUnitNode}
   > get_swing_axis = RigVMCollapseNode(
    Axis to Get=(X=0,Y=1,Z=1)
    Backward=false
    Forward=true
    Inward=false
    Outward=true
    Rotation Order=ZXY
    Twist Axis=(X=1,Y=0,Z=0)
    input joint=thigh_l_constraint
   )
   > &RigUnit_ModifyTransforms(
    ItemToModify=((Item=(Type=Bone,Name="thigh_l_support"),Transform=(Rotation=(X=0,Y=0,Z=0,W=1),Translation=(X=0,Y=0,Z=0),Scale3D=(X=1,Y=1,Z=1))))
    Weight=1
    WeightMinimum=0
    ItemToModify.0.Transform.Rotation=@get_swing_axis.Output Rotate.Value
   )
   > &RigUnit_ParentConstraint(
    AdvancedSettings=(InterpolationType=Average,RotationOrderForFilter=XZY)
    Child=(Type=Bone,Name="base_controller")
    Parents=((Item=(Type=Bone,Name="thigh_l_support"),Weight=1))
    Weight=1
   )
   > RigVMFunctionReturnNode()
  ]}
 New Function_ContainedGraph {nodes [
   n1 = RigVMFunctionEntryNode()
   RigUnit_GetTransform = &RigUnit_GetTransform(Item=(Type=Bone,Name="thigh_l_for_cons"), Space=LocalSpace, bInitial=false, Item.Name=@n1.input joint)
   RigVMFunction_MathQuaternionToEuler = &RigVMFunction_MathQuaternionToEuler(
    RotationOrder=@n1.Rotation Order
    Value=&RigVMFunction_MathQuaternionSwingTwist(
     TwistAxis=(X=1,Y=0,Z=0)
     TwistAxis.X=@n1.Twist Axis.X
     TwistAxis.Y=@n1.Twist Axis.Y
     TwistAxis.Z=@n1.Twist Axis.Z
     Input=@RigUnit_GetTransform.Transform.Rotation
    ).Swing
   )
   n2 = &RigVMFunction_MathDoubleMul(B=@RigVMFunction_MathQuaternionToEuler.Result.Z, A=@n1.Axis to Get.Z)
   n3 = /Script/RigVM.RigVMFunction_MathDoubleLess(B=0, A=@n2.Result) {is=struct RigVMUnitNode, template="Less::Execute(in A,in B,out Result)"}
   n4 = /Script/RigVM.RigVMFunction_MathDoubleGreater(B=0, A=@n2.Result) {is=struct RigVMUnitNode, template="Greater::Execute(in A,in B,out Result)"}
   Add = &RigVMFunction_MathDoubleAdd(
    A=RigVMDispatchNode(
     False=0
     Condition=@n1.Forward
     True=RigVMDispatchNode(False=0, Condition=@n3.Result, True=@RigVMFunction_MathQuaternionToEuler.Result.Z) {template="DISPATCH_RigVMDispatch_If(in Condition,in True,in False,out Result)"}.Result
    ) {template="DISPATCH_RigVMDispatch_If(in Condition,in True,in False,out Result)"}.Result
    B=RigVMDispatchNode(
     False=0
     Condition=@n1.Backward
     True=RigVMDispatchNode(False=0, Condition=@n4.Result, True=@RigVMFunction_MathQuaternionToEuler.Result.Z) {template="DISPATCH_RigVMDispatch_If(in Condition,in True,in False,out Result)"}.Result
    ) {template="DISPATCH_RigVMDispatch_If(in Condition,in True,in False,out Result)"}.Result
   )
   n5 = &RigVMFunction_MathDoubleMul(A=@n1.Axis to Get.Y, B=@RigVMFunction_MathQuaternionToEuler.Result.Y)
   Less = &RigVMFunction_MathDoubleLess(B=0, A=@n5.Result) {template="Less::Execute(in A,in B,out Result)"}
   n6 = RigVMDispatchNode(False=0, Condition=@Less.Result, True=@RigVMFunction_MathQuaternionToEuler.Result.Y) {template="DISPATCH_RigVMDispatch_If(in Condition,in True,in False,out Result)"}
   Greater = &RigVMFunction_MathDoubleGreater(B=0, A=@n5.Result) {template="Greater::Execute(in A,in B,out Result)"}
   n7 = RigVMDispatchNode(False=0, Condition=@Greater.Result, True=@RigVMFunction_MathQuaternionToEuler.Result.Y) {template="DISPATCH_RigVMDispatch_If(in Condition,in True,in False,out Result)"}
   RigVMFunction_MathQuaternionFromEuler = &RigVMFunction_MathQuaternionFromEuler(
    Euler=(X=0,Y=0,Z=0)
    Euler.Z=@Add.Result
    Euler.Y=&RigVMFunction_MathDoubleAdd(
     A=RigVMDispatchNode(False=0, Condition=@n1.Outward, True=@n6.Result) {template="DISPATCH_RigVMDispatch_If(in Condition,in True,in False,out Result)"}.Result
     B=RigVMDispatchNode(False=0, Condition=@n1.Inward, True=@n7.Result) {template="DISPATCH_RigVMDispatch_If(in Condition,in True,in False,out Result)"}.Result
    ).Result
    RotationOrder=@n1.Rotation Order
   )
   > RigVMFunctionReturnNode(Output Rotate=(Value=(X=0,Y=0,Z=0,W=1)), Output Rotate.Value=@RigVMFunction_MathQuaternionFromEuler.Result)
  ]}
 New Function_ContainedGraph {nodes [
   RigVMFunctionEntryNode()
   > n1 = &RigUnit_SphericalPoseReader(
    ActiveRegionScaleFactors=(PositiveWidth=1,NegativeWidth=1,PositiveHeight=1,NegativeHeight=1)
    ActiveRegionSize=0.05
    Debug=(bDrawDebug=false,bDraw2D=False,bDrawLocalAxes=False,DebugScale=25,DebugSegments=20,DebugThickness=0.25)
    DriverAxis=(X=-1,Y=0,Z=0)
    DriverItem=(Type=Bone,Name="thigh_l")
    FalloffRegionScaleFactors=(PositiveWidth=1,NegativeWidth=1,PositiveHeight=1,NegativeHeight=1)
    FalloffSize=0.49
    FlipHeightScaling=False
    FlipWidthScaling=False
    OptionalParentItem=(Type=Bone,Name="None")
    RotationOffset=(X=90,Y=0,Z=0)
   )
   > RigVMVariableNode(Value=@n1.OutputParam) {variable=leg forward}
   > &RigUnit_OffsetTransformForItem(
    Item=(Type=Bone,Name="base_controller")
    OffsetTransform=(Rotation=(X=0.0949,Y=-0.1115,Z=-0.3397,W=-0.9291),Translation=(X=-7.5703,Y=3.9449,Z=-8.989),Scale3D=(X=1,Y=1,Z=1))
    bPropagateToChildren=True
    Weight=@n1.OutputParam
   )
   > &RigUnit_OffsetTransformForItem(
    Item=(Type=Bone,Name="base_belt_helper_back")
    OffsetTransform=(Rotation=(X=0,Y=0,Z=-0.0874,W=-0.9962),Translation=(X=0,Y=0,Z=0),Scale3D=(X=1,Y=1,Z=1))
    bPropagateToChildren=True
    Weight=@n1.OutputParam
   )
   > &RigUnit_OffsetTransformForItem(
    Item=(Type=Bone,Name="base_belt_helper_front")
    OffsetTransform=(Rotation=(X=0.0151,Y=0.173,Z=-0.0858,W=0.9811),Translation=(X=0,Y=0,Z=0),Scale3D=(X=1,Y=1,Z=1))
    bPropagateToChildren=True
    Weight=@n1.OutputParam
   )
   > n2 = &RigUnit_SphericalPoseReader(
    ActiveRegionScaleFactors=(PositiveWidth=1,NegativeWidth=1,PositiveHeight=1,NegativeHeight=1)
    ActiveRegionSize=0.05
    Debug=(bDrawDebug=false,bDraw2D=False,bDrawLocalAxes=False,DebugScale=25,DebugSegments=20,DebugThickness=0.25)
    DriverAxis=(X=-1,Y=0,Z=0)
    DriverItem=(Type=Bone,Name="thigh_l")
    FalloffRegionScaleFactors=(PositiveWidth=1,NegativeWidth=1,PositiveHeight=1,NegativeHeight=1)
    FalloffSize=0.49
    FlipHeightScaling=False
    FlipWidthScaling=False
    OptionalParentItem=(Type=Bone,Name="None")
    RotationOffset=(X=180,Y=0,Z=0)
   )
   > RigVMVariableNode(Value=@n2.OutputParam) {variable=leg side}
   > &RigUnit_OffsetTransformForItem(
    Item=(Type=Bone,Name="base_controller")
    OffsetTransform=(Rotation=(X=0,Y=0,Z=0.165,W=-0.9863),Translation=(X=-4.3674,Y=2.5458,Z=-6.4288),Scale3D=(X=1,Y=1,Z=1))
    bPropagateToChildren=True
    Weight=@n2.OutputParam
   )
   > &RigUnit_OffsetTransformForItem(
    Item=(Type=Bone,Name="base_belt_helper_back")
    OffsetTransform=(Rotation=(X=0,Y=0,Z=0.1067,W=0.9943),Translation=(X=0,Y=0,Z=0),Scale3D=(X=1,Y=1,Z=1))
    bPropagateToChildren=True
    Weight=@n2.OutputParam
   )
   > &RigUnit_OffsetTransformForItem(
    Item=(Type=Bone,Name="base_belt_helper_front")
    OffsetTransform=(Rotation=(X=0,Y=0.2263,Z=0,W=-0.9741),Translation=(X=-0.2936,Y=0,Z=-0.5326),Scale3D=(X=1,Y=1,Z=0.903))
    bPropagateToChildren=True
    Weight=@n2.OutputParam
   )
   > n3 = &RigUnit_SphericalPoseReader(
    ActiveRegionScaleFactors=(PositiveWidth=1,NegativeWidth=1,PositiveHeight=1,NegativeHeight=1)
    ActiveRegionSize=0.05
    Debug=(bDrawDebug=false,bDraw2D=False,bDrawLocalAxes=False,DebugScale=25,DebugSegments=20,DebugThickness=0.25)
    DriverAxis=(X=-1,Y=0,Z=0)
    DriverItem=(Type=Bone,Name="thigh_l")
    FalloffRegionScaleFactors=(PositiveWidth=1,NegativeWidth=1,PositiveHeight=1,NegativeHeight=1)
    FalloffSize=0.49
    FlipHeightScaling=False
    FlipWidthScaling=False
    OptionalParentItem=(Type=Bone,Name="None")
    RotationOffset=(X=135,Y=0,Z=0)
   )
   > RigVMVariableNode(Value=@n3.OutputParam) {variable=inbetween front side}
   > &RigUnit_OffsetTransformForItem(
    Item=(Type=Bone,Name="base_controller")
    OffsetTransform=(Rotation=(X=-0.1871,Y=-0.1954,Z=-0.1864,W=0.9445),Translation=(X=6.8292,Y=-10.7661,Z=4.9078),Scale3D=(X=1,Y=1,Z=1))
    bPropagateToChildren=True
    Weight=@n3.OutputParam
   )
   > &RigUnit_OffsetTransformForItem(
    Item=(Type=Bone,Name="base_belt_helper_back")
    OffsetTransform=(Rotation=(X=0,Y=0,Z=-0.1114,W=0.9938),Translation=(X=0,Y=0,Z=0),Scale3D=(X=1,Y=1,Z=1))
    bPropagateToChildren=True
    Weight=@n3.OutputParam
   )
   > n4 = &RigUnit_SphericalPoseReader(
    ActiveRegionScaleFactors=(PositiveWidth=1,NegativeWidth=1,PositiveHeight=1,NegativeHeight=1)
    ActiveRegionSize=0.02
    Debug=(bDrawDebug=false,bDraw2D=False,bDrawLocalAxes=False,DebugScale=25,DebugSegments=20,DebugThickness=0.25)
    DriverAxis=(X=-1,Y=0,Z=0)
    DriverItem=(Type=Bone,Name="thigh_l")
    FalloffRegionScaleFactors=(PositiveWidth=1,NegativeWidth=2.8,PositiveHeight=1,NegativeHeight=1)
    FalloffSize=0.1
    FlipHeightScaling=False
    FlipWidthScaling=False
    OptionalParentItem=(Type=Bone,Name="None")
    RotationOffset=(X=-90,Y=0,Z=-45)
   )
   > &RigUnit_OffsetTransformForItem(
    Item=(Type=Bone,Name="base_belt_helper_back")
    OffsetTransform=(Rotation=(X=0.0076,Y=-0.043,Z=0.1735,W=0.9839),Translation=(X=0,Y=0,Z=0),Scale3D=(X=1,Y=1,Z=1))
    bPropagateToChildren=True
    Weight=@n4.OutputParam
   )
   > &RigUnit_ModifyTransforms(
    ItemToModify=((Item=(Type=Bone,Name="base_main_chain"),Transform=(Rotation=(X=0,Y=0,Z=-0.3939,W=0.9192),Translation=(X=0,Y=0,Z=0),Scale3D=(X=1,Y=1,Z=1))))
    WeightMinimum=0
    Weight=@n4.OutputParam
   )
   > &RigUnit_AimConstraintLocalSpaceOffset(
    Child=(Type=Bone,Name="base_belt_helper_front")
    Filter=(bX=True,bY=True,bZ=True)
    Parents=((Item=(Type=Bone,Name="for_aim_belt_front"),Weight=1))
    UpAxis=(X=0,Y=1,Z=0)
    Weight=1
    WorldUp=(Target=(X=0,Y=0,Z=0),Kind=Direction,Space=(Type=Bone,Name="None"))
   )
   > RigVMFunctionReturnNode()
  ]}
 New Function_ContainedGraph {nodes [
   RigVMFunctionEntryNode()
   > n1 = &RigUnit_SphericalPoseReader(
    ActiveRegionScaleFactors=(PositiveWidth=1,NegativeWidth=1,PositiveHeight=1,NegativeHeight=1)
    ActiveRegionSize=0.15
    Debug=(bDrawDebug=false,bDraw2D=False,bDrawLocalAxes=False,DebugScale=25,DebugSegments=20,DebugThickness=0.25)
    DriverAxis=(X=1,Y=0,Z=0)
    DriverItem=(Type=Bone,Name="upperarm_l")
    FalloffRegionScaleFactors=(PositiveWidth=1,NegativeWidth=1,PositiveHeight=1,NegativeHeight=1)
    FalloffSize=0.25
    FlipHeightScaling=False
    FlipWidthScaling=False
    OptionalParentItem=(Type=Bone,Name="None")
    RotationOffset=(X=0,Y=-45,Z=20)
   )
   > &RigUnit_ParentConstraint(
    AdvancedSettings=(InterpolationType=Average,RotationOrderForFilter=XZY)
    Child=(Type=Bone,Name="spine_02_for_cons")
    Filter=(TranslationFilter=(bX=True,bY=True,bZ=True),RotationFilter=(bX=true,bY=true,bZ=true),ScaleFilter=(bX=True,bY=True,bZ=True))
    Parents=((Item=(Type=Bone,Name="spine_02"),Weight=1))
    Weight=1
   )
   n2 = &RigVMFunction_AlphaInterp(
    InterpSpeedDecreasing=10
    InterpSpeedIncreasing=10
    Scale=1
    bClampResult=False
    Value=RigVMDispatchNode(False=1, True=0, Condition=RigVMVariableNode() {variable="sliding  or rail"}.Value) {template="DISPATCH_RigVMDispatch_If(in Condition,in True,in False,out Result)"}.Result
   )
   RigVMFunction_AnimEvalRichCurve = &RigVMFunction_AnimEvalRichCurve(
    Curve="(EditorCurveData=(Keys=((InterpMode=RCIM_Cubic,TangentMode=RCTM_Auto,TangentWeightMode=RCTWM_WeightedNone,Time=0,Value=0,ArriveTangent=0,ArriveTangentWeight=0,LeaveTangent=0,LeaveTangentWeight=0),(InterpMode=RCIM_Cubic,TangentMode=RCTM_Auto,TangentWeightMo ...(truncated)"
    SourceMinimum=0
    TargetMinimum=0
    Value=@n2.Result
   )
   Cast = &RigVMFunction_MathBoolToFloat(Value=RigVMVariableNode() {variable=targeting}.Value)
   Multiply = &RigVMFunction_MathDoubleMul(B=@Cast.Result, A=@n1.OutputParam)
   AlphaInterp = &RigVMFunction_AlphaInterp(InterpSpeedDecreasing=10, InterpSpeedIncreasing=10, Scale=1, bClampResult=false, Value=@Multiply.Result)
   n3 = &RigVMFunction_AnimEvalRichCurve(
    Curve="(EditorCurveData=(Keys=((InterpMode=RCIM_Cubic,TangentMode=RCTM_Auto,TangentWeightMode=RCTWM_WeightedNone,Time=0,Value=0,ArriveTangent=0,ArriveTangentWeight=0,LeaveTangent=0,LeaveTangentWeight=0),(InterpMode=RCIM_Cubic,TangentMode=RCTM_Auto,TangentWeightMo ...(truncated)"
    SourceMinimum=0
    TargetMinimum=0
    Value=@AlphaInterp.Result
   )
   RigUnit_GetTransform = &RigUnit_GetTransform(Item=(Type=Bone,Name="spine_02_for_cons"), Space=LocalSpace, bInitial=false)
   RigVMFunction_MathQuaternionToEuler = &RigVMFunction_MathQuaternionToEuler(RotationOrder=ZYX, Value=@RigUnit_GetTransform.Transform.Rotation)
   RigVMFunction_MathQuaternionFromEuler = &RigVMFunction_MathQuaternionFromEuler(Euler=/Script/RigVM.RigVMFunction_MathVectorMul(
     B=(X=-1,Y=-1,Z=-1)
     B.X=&RigVMFunction_MathDoubleRemap(SourceMaximum=1, SourceMinimum=0, TargetMaximum=1, TargetMinimum=-1, Value=@n3.Result).Result
     A=@RigVMFunction_MathQuaternionToEuler.Result
    ) {is=struct RigVMUnitNode, template="Multiply::Execute(in A,in B,out Result)"}.Result)
   > Set Transform = &RigUnit_SetRotation(
    Item=(Type=Bone,Name="reverse_spine_02")
    bInitial=false
    Weight=@RigVMFunction_AnimEvalRichCurve.Result
    Value=@RigVMFunction_MathQuaternionFromEuler.Result
   )
   > &RigUnit_RotationConstraintLocalSpaceOffset(
    AdvancedSettings=(InterpolationType=Average,RotationOrderForFilter=XZY)
    Child=(Type=Bone,Name="reverse_spine_02")
    Filter=(bX=True,bY=True,bZ=True)
    Parents=((Item=(Type=Bone,Name="pelvis"),Weight=1))
    Weight=RigVMVariableNode() {variable=Idle Animation FrontEnd}.Value
   )
   > RigVMFunctionReturnNode()
  ]}
 New Function_ContainedGraph {nodes [
   RigVMFunctionEntryNode()
   RigVMFunction_AnimEvalRichCurve = &RigVMFunction_AnimEvalRichCurve(
    Curve="(EditorCurveData=(Keys=((InterpMode=RCIM_Cubic,TangentMode=RCTM_Auto,TangentWeightMode=RCTWM_WeightedNone,Time=0,Value=0,ArriveTangent=0,ArriveTangentWeight=0,LeaveTangent=0,LeaveTangentWeight=0),(InterpMode=RCIM_Cubic,TangentMode=RCTM_Auto,TangentWeightMo ...(truncated)"
    SourceMinimum=0
    TargetMinimum=0
    Value=RigVMVariableNode() {variable=leg forward}.Value
   )
   n1 = RigVMVariableNode() {variable=targeting}
   AlphaInterp = &RigVMFunction_AlphaInterp(
    InterpSpeedDecreasing=10
    InterpSpeedIncreasing=10
    Scale=1
    bClampResult=false
    bMapRange=false
    Value=RigVMDispatchNode(True=0, False=@RigVMFunction_AnimEvalRichCurve.Result, Condition=@n1.Value) {template="DISPATCH_RigVMDispatch_If(in Condition,in True,in False,out Result)"}.Result
   )
   > &RigUnit_OffsetTransformForItem(
    Item=(Type=Bone,Name="base_jacket_l_0")
    OffsetTransform=(Rotation=(X=-0.2626,Y=-0.0981,Z=0.287,W=-0.916),Translation=(X=0,Y=0,Z=0),Scale3D=(X=1,Y=1,Z=1))
    bPropagateToChildren=True
    Weight=@AlphaInterp.Result
   )
   > &RigUnit_OffsetTransformForItem(
    Item=(Type=Bone,Name="base_jacket_l_1")
    OffsetTransform=(Rotation=(X=-0.1539,Y=-0.1316,Z=0.059,W=-0.9775),Translation=(X=0,Y=0,Z=0),Scale3D=(X=1,Y=1,Z=1))
    bPropagateToChildren=True
    Weight=@AlphaInterp.Result
   )
   > &RigUnit_OffsetTransformForItem(
    Item=(Type=Bone,Name="base_jacket_l_2")
    OffsetTransform=(Rotation=(X=0.0868,Y=-0.0076,Z=-0.0868,W=0.9924),Translation=(X=0,Y=0,Z=0),Scale3D=(X=1,Y=1,Z=1))
    bPropagateToChildren=true
    Weight=@AlphaInterp.Result
   )
   > &RigUnit_OffsetTransformForItem(
    Item=(Type=Bone,Name="base_joystick")
    OffsetTransform=(Rotation=(X=-0.0484,Y=0.1754,Z=0.4716,W=0.8628),Translation=(X=0,Y=0,Z=0),Scale3D=(X=1,Y=1,Z=1))
    bPropagateToChildren=True
    Weight=@AlphaInterp.Result
   )
   n2 = &RigVMFunction_AnimEvalRichCurve(
    Curve="(EditorCurveData=(Keys=((InterpMode=RCIM_Cubic,TangentMode=RCTM_Auto,TangentWeightMode=RCTWM_WeightedNone,Time=0,Value=0,ArriveTangent=0,ArriveTangentWeight=0,LeaveTangent=0,LeaveTangentWeight=0),(InterpMode=RCIM_Cubic,TangentMode=RCTM_User,TangentWeightMo ...(truncated)"
    SourceMinimum=0
    TargetMinimum=0
    Value=RigVMVariableNode() {variable=leg side}.Value
   )
   n3 = &RigVMFunction_AlphaInterp(
    InterpSpeedDecreasing=10
    InterpSpeedIncreasing=10
    Scale=1
    bClampResult=false
    bMapRange=false
    Value=RigVMDispatchNode(True=0, False=@n2.Result, Condition=@n1.Value) {template="DISPATCH_RigVMDispatch_If(in Condition,in True,in False,out Result)"}.Result
   )
   > &RigUnit_OffsetTransformForItem(
    Item=(Type=Bone,Name="base_jacket_l_0")
    OffsetTransform=(Rotation=(X=0,Y=0,Z=0.4934,W=-0.8698),Translation=(X=0,Y=0,Z=0),Scale3D=(X=1,Y=1,Z=1))
    bPropagateToChildren=True
    Weight=@n3.Result
   )
   > &RigUnit_OffsetTransformForItem(
    Item=(Type=Bone,Name="base_jacket_l_1")
    OffsetTransform=(Rotation=(X=-0.0144,Y=0.0101,Z=-0.5733,W=0.8192),Translation=(X=0,Y=0,Z=0),Scale3D=(X=1,Y=1,Z=1))
    bPropagateToChildren=True
    Weight=@n3.Result
   )
   > &RigUnit_OffsetTransformForItem(
    Item=(Type=Bone,Name="base_jacket_l_2")
    OffsetTransform=(Rotation=(X=0.171,Y=-0.0302,Z=-0.171,W=0.9698),Translation=(X=0,Y=0,Z=0),Scale3D=(X=1,Y=1,Z=1))
    bPropagateToChildren=True
    Weight=@n3.Result
   )
   n4 = &RigVMFunction_AlphaInterp(
    InterpSpeedDecreasing=10
    InterpSpeedIncreasing=10
    Scale=1
    bClampResult=false
    bMapRange=false
    Value=RigVMDispatchNode(True=0, False=RigVMVariableNode() {variable=inbetween front side}.Value, Condition=@n1.Value) {template="DISPATCH_RigVMDispatch_If(in Condition,in True,in False,out Result)"}.Result
   )
   > &RigUnit_OffsetTransformForItem(
    Item=(Type=Bone,Name="base_jacket_l_0")
    OffsetTransform=(Rotation=(X=-0.0189,Y=-0.2156,Z=-0.0851,W=0.9726),Translation=(X=0,Y=0,Z=0),Scale3D=(X=1,Y=1,Z=1))
    bPropagateToChildren=True
    Weight=@n4.Result
   )
   > &RigUnit_OffsetTransformForItem(
    Item=(Type=Bone,Name="base_jacket_l_1")
    OffsetTransform=(Rotation=(X=-0.25,Y=-0.067,Z=0.25,W=0.933),Translation=(X=0,Y=0,Z=0),Scale3D=(X=1,Y=1,Z=1))
    bPropagateToChildren=True
    Weight=@n4.Result
   )
   Cast = &RigVMFunction_MathBoolToFloat(Value=RigVMVariableNode() {variable=crouching}.Value)
   Remap = &RigVMFunction_MathDoubleRemap(
    SourceMaximum=44.5
    SourceMinimum=70
    TargetMaximum=1
    TargetMinimum=0
    Value=&RigUnit_GetRelativeTransformForItem(Child=(Type=Bone,Name="pelvis"), Parent=(Type=Bone,Name="root"), bChildInitial=False, bParentInitial=False).RelativeTransform.Translation.Z
   )
   Multiply = &RigVMFunction_MathDoubleMul(B=@Cast.Result, A=@Remap.Result)
   > &RigUnit_ModifyTransforms(
    ItemToModify="((Item=(Type=Bone,Name=\"base_jacket_l_0\"),Transform=(Rotation=(X=0.264,Y=0.2147,Z=0.2015,W=0.9185),Translation=(X=3.087,Y=-1.3537,Z=0.0734),Scale3D=(X=1,Y=1,Z=1))),(Item=(Type=Bone,Name=\"base_jacket_l_1\"),Transform=(Rotation=(X=0.1721,Y=-0.0491,Z=0.0373,W= ...(truncated)"
    WeightMinimum=0
    Weight=@Multiply.Result
   )
   Subtract = &RigVMFunction_MathDoubleSub(B=1, A=@Multiply.Result)
   Absolute = &RigVMFunction_MathDoubleAbs(Value=@Subtract.Result)
   > /Script/ControlRig.RigUnit_SetCurveValue(Curve=crouching, Value=@Absolute.Result) {is=struct RigVMUnitNode}
   n5 = &RigVMFunction_AnimEvalRichCurve(
    Curve=(EditorCurveData=(Keys=((),(Time=0.5,Value=1),(Time=0.8))))
    SourceMaximum=2
    SourceMinimum=0
    TargetMinimum=0
    Value=&RigVMFunction_MathDoubleRemap(
     SourceMaximum=70
     SourceMinimum=45
     TargetMaximum=1
     TargetMinimum=0
     Value=&RigUnit_GetRelativeTransformForItem(Child=(Type=Bone,Name="pelvis"), Parent=(Type=Bone,Name="root"), bChildInitial=False, bParentInitial=False).RelativeTransform.Translation.Z
    ).Result
   )
   Multiply_1 = RigVMAggregateNode(A=@n5.Result, C=RigVMVariableNode() {variable=leg forward}.Value, B=@Cast.Result)
   > &RigUnit_ModifyTransforms(
    ItemToModify="((Item=(Type=Bone,Name=\"base_jacket_l_0\"),Transform=(Rotation=(X=0.2072,Y=-0.1925,Z=-0.2372,W=0.9294),Translation=(X=-0.357,Y=-1.567,Z=-0.8512),Scale3D=(X=1,Y=1,Z=1))),(Item=(Type=Bone,Name=\"base_jacket_l_1\"),Transform=(Rotation=(X=-0.069,Y=-0.0491,Z=-0.08 ...(truncated)"
    WeightMinimum=0
    Weight=@Multiply_1.Result
   )
   > &RigUnit_SetCurveValue(
    Curve=crouching sprinting
    Value=&RigVMFunction_MathDoubleAbs(Value=&RigVMFunction_MathDoubleSub(B=1, A=@Multiply_1.Result).Result).Result
   )
   Add = &RigVMFunction_MathDoubleAdd(A=RigVMVariableNode() {variable=leg forward}.Value, B=RigVMVariableNode() {variable=inbetween front side}.Value)
   > &RigUnit_ModifyTransforms(
    ItemToModify=((Item=(Type=Bone,Name="base_belt_helper_back"),Transform=(Rotation=(X=-0.0069,Y=-0.0891,Z=0.0967,W=0.9913),Translation=(X=0.032,Y=0.0183,Z=-0.1193),Scale3D=(X=1,Y=1,Z=1))))
    WeightMinimum=0
    Weight=&RigVMFunction_MathDoubleRemap(SourceMaximum=0.45, SourceMinimum=0, TargetMaximum=1, TargetMinimum=0, Value=@Add.Result).Result
   )
   > DeltaRotation_6 = RigVMFunctionReferenceNode(
    Rotation Order=ZYX
    child=spine_01
    debugRotation=false
    max=-30
    min=0
    parent=spine_01_null
    parentType=Null
    premult=1
    primAxis=2
    twistAxis=(X=1,Y=0,Z=0)
   )
   n6 = &RigVMFunction_AnimEvalRichCurve(
    Curve="(EditorCurveData=(Keys=((InterpMode=RCIM_Cubic,TangentMode=RCTM_User,TangentWeightMode=RCTWM_WeightedNone,Time=0,Value=0,ArriveTangent=4.3694,ArriveTangentWeight=0,LeaveTangent=4.3693,LeaveTangentWeight=0),(InterpMode=RCIM_Cubic,TangentMode=RCTM_Auto,Tange ...(truncated)"
    SourceMinimum=0
    TargetMinimum=0
    Value=@DeltaRotation_6.outputWeight
   )
   n7 = &RigVMFunction_AlphaInterp(InterpSpeedDecreasing=10, InterpSpeedIncreasing=10, Scale=1, bClampResult=true, Value=@n6.Result)
   > &RigUnit_ModifyTransforms(
    ItemToModify="((Item=(Type=Bone,Name=\"base_jacket_b\"),Transform=(Rotation=(X=0,Y=0,Z=-0.1736,W=0.9848),Translation=(X=-1,Y=0,Z=0),Scale3D=(X=1,Y=1,Z=1))),(Item=(Type=Bone,Name=\"base_jacket_l_2\"),Transform=(Rotation=(X=0,Y=0,Z=-0.1736,W=0.9848),Translation=(X=-1,Y=0,Z=0) ...(truncated)"
    WeightMinimum=0
    Weight=@n7.Result
   )
   > RigVMFunctionReturnNode()
  ]}
 New Function_ContainedGraph {nodes [
   RigVMFunctionEntryNode()
   Remap = &RigVMFunction_MathDoubleRemap(
    SourceMaximum=1
    SourceMinimum=0.7
    TargetMaximum=1
    TargetMinimum=0
    Value=RigVMDispatchNode(BoneName=root, Default=0, Name=deformRig_l_knee_alpha) {template="DISPATCH_RigDispatch_GetAnimAttribute(in Name,in BoneName,in Default,out Value,out Found)"}.Value
   )
   AlphaInterp = &RigVMFunction_AlphaInterp(InterpSpeedDecreasing=15, InterpSpeedIncreasing=15, Scale=1, bClampResult=false, bMapRange=false, Value=@Remap.Result)
   RigVMFunction_AnimEvalRichCurve = &RigVMFunction_AnimEvalRichCurve(
    Curve="(EditorCurveData=(Keys=((InterpMode=RCIM_Cubic,TangentMode=RCTM_Auto,TangentWeightMode=RCTWM_WeightedNone,Time=0,Value=0,ArriveTangent=0,ArriveTangentWeight=0,LeaveTangent=0,LeaveTangentWeight=0),(InterpMode=RCIM_Cubic,TangentMode=RCTM_Auto,TangentWeightMo ...(truncated)"
    SourceMinimum=0
    TargetMinimum=0
    Value=@AlphaInterp.Result
   )
   > &RigUnit_ModifyTransforms(
    ItemToModify=((Item=(Type=Bone,Name="dyn_cube_leg_l"),Transform=(Rotation=(X=-0.0046,Y=-0.0055,Z=0.1331,W=0.9911),Translation=(X=-0.15,Y=4.5,Z=3.6),Scale3D=(X=1,Y=1,Z=1))))
    WeightMinimum=0
    Weight=@RigVMFunction_AnimEvalRichCurve.Result
   )
   n1 = &RigVMFunction_AlphaInterp(
    InterpSpeedDecreasing=15
    InterpSpeedIncreasing=15
    Scale=1
    bClampResult=false
    bMapRange=true
    Value=&RigVMFunction_MathDoubleRemap(
     SourceMaximum=1
     SourceMinimum=0.5
     TargetMaximum=1
     TargetMinimum=0
     Value=RigVMDispatchNode(BoneName=root, Default=0, Name=deformRig_r_knee_alpha) {template="DISPATCH_RigDispatch_GetAnimAttribute(in Name,in BoneName,in Default,out Value,out Found)"}.Value
    ).Result
   )
   n2 = &RigVMFunction_AnimEvalRichCurve(
    Curve="(EditorCurveData=(Keys=((InterpMode=RCIM_Cubic,TangentMode=RCTM_Auto,TangentWeightMode=RCTWM_WeightedNone,Time=0,Value=0,ArriveTangent=0,ArriveTangentWeight=0,LeaveTangent=0,LeaveTangentWeight=0),(InterpMode=RCIM_Cubic,TangentMode=RCTM_Auto,TangentWeightMo ...(truncated)"
    SourceMinimum=0
    TargetMinimum=0
    Value=@n1.Result
   )
   > &RigUnit_ModifyTransforms(
    ItemToModify=((Item=(Type=Bone,Name="dyn_cube_leg_r"),Transform=(Rotation=(X=0.0012,Y=0.0059,Z=0.3224,W=-0.9466),Translation=(X=-0.6748,Y=-6.2479,Z=4.163),Scale3D=(X=1,Y=1,Z=1))))
    WeightMinimum=0
    Weight=@n2.Result
   )
   > RigVMFunctionReturnNode()
  ]}
 New Function_ContainedGraph {nodes [
   n1 = RigVMFunctionEntryNode()
   RigVMFunction_MathBoolToInteger = &RigVMFunction_MathBoolToInteger(Value=@n1.onlyTwist)
   RigVMFunction_MathQuaternionSwingTwist = &RigVMFunction_MathQuaternionSwingTwist(
    twistAxis=@n1.twistAxis
    Input=&RigUnit_GetRelativeTransformForItem(
     bChildInitial=False
     bParentInitial=False
     child=(Type=Bone,Name="None")
     parent=(Type=Bone,Name="None")
     child.Name=@n1.child
     parent.Name=@n1.parent
     parent.Type=@n1.parentType
    ).RelativeTransform.Rotation
   )
   RigVMFunction_MathQuaternionToEuler = &RigVMFunction_MathQuaternionToEuler(
    Value=RigVMDispatchNode(
     Values=((),())
     Index=@RigVMFunction_MathBoolToInteger.Result
     Values.0=@RigVMFunction_MathQuaternionSwingTwist.Swing
     Values.1=@RigVMFunction_MathQuaternionSwingTwist.Twist
    ) {template="DISPATCH_RigVMDispatch_SelectInt32(in Index,in Values,out Result)"}.Result
    RotationOrder=@n1.Rotation Order
   )
   > DISPATCH_RigVMDispatch_Print = RigVMDispatchNode(ScreenColor=(R=1,G=1,B=1,A=1), ScreenDuration=1.5, Enabled=@n1.debugRotation, Value=@RigVMFunction_MathQuaternionToEuler.Result) {template="DISPATCH_RigVMDispatch_Print(in Prefix,in Value,in Enabled,in ScreenDuration,in ScreenColor)"}
   Clamp = &RigVMFunction_MathIntClamp(Value=@n1.primAxis)
   DISPATCH_RigVMDispatch_SelectInt32 = RigVMDispatchNode(
    Values=(0,0,0)
    Index=@Clamp.Result
    Values.0=@RigVMFunction_MathQuaternionToEuler.Result.X
    Values.1=@RigVMFunction_MathQuaternionToEuler.Result.Y
    Values.2=@RigVMFunction_MathQuaternionToEuler.Result.Z
   ) {template="DISPATCH_RigVMDispatch_SelectInt32(in Index,in Values,out Result)"}
   Remap = &RigVMFunction_MathDoubleRemap(
    TargetMaximum=1
    TargetMinimum=0
    bClamp=False
    Value=@DISPATCH_RigVMDispatch_SelectInt32.Result
    SourceMaximum=@n1.max
    SourceMinimum=@n1.min
   )
   Multiply = &RigVMFunction_MathDoubleMul(A=&RigVMFunction_MathDoubleClamp(Value=@Remap.Result).Result, B=@n1.premult)
   > RigVMFunctionReturnNode(outputWeight=@Multiply.Result)
  ]}
 RigVMFunctionLibrary {nodes [
   DeltaRotation = RigVMCollapseNode(ExecuteContext=(), Rotation Order=ZYX, premult=1, twistAxis=(X=1,Y=0,Z=0))
   controller setup = RigVMCollapseNode(ExecuteContext=())
   jacket setup = RigVMCollapseNode(ExecuteContext=())
   leg cubes setup = RigVMCollapseNode(ExecuteContext=())
   reverse rotation jacket = RigVMCollapseNode(ExecuteContext=())
   surface constraints = RigVMCollapseNode(ExecuteContext=())
   swing setup for controller = RigVMCollapseNode(ExecuteContext=())
  ]}
 RigVMModel {nodes [
   RigUnit_BeginExecution = &RigUnit_BeginExecution()
   FN_DeformRig_FNC = RigVMFunctionReferenceNode(
    DeformRig_ Curve="(deformRig_l_upperarm_up_alpha_72_#1=False,deformRig_r_upperarm_up_alpha_73_#2=False,deformRig_l_upperarm_fr_alpha_74_#3=False,deformRig_r_upperarm_fr_alpha_75_CB51B8 ...(truncated)"
    DeformRig_Bool="(upperarm_1_#4=True,elbow_3_#5=True,wrist_7_#6=True,groin_8_#7=True,glute_17_#8=True,knee_9_40C408324CA261 ...(truncated)"
    DeformRig_Debug=(upperarm_debug_18_#4=False,wrist_debug_20_#6=False,lat_debug_26_#9=False,groinGlute_debug_27_#7=False)
    DeformRig_Value="(upperarm_up_18_#10=(Rotation=(X=0.0,Y=-0.0,Z=0.0,W=1.0),Translation=(X=0.0,Y=0.0,Z=8.0),Scale3D=(X=0.70,Y=1.0,Z=1.0)),upperarm_fwd_fr_19_#11=(Rotation=(X=0.0,Y=-0.0,Z=0.0,W=1.0),Translation=(X=0.0, ...(truncated)"
    ExecuteContext=@RigUnit_BeginExecution.ExecutePin
   )
   > surface constraints = RigVMFunctionReferenceNode()
   > swing setup for controller = RigVMFunctionReferenceNode()
   > controller setup = RigVMFunctionReferenceNode()
   > reverse rotation jacket = RigVMFunctionReferenceNode()
   > jacket setup = RigVMFunctionReferenceNode()
   > leg cubes setup = RigVMFunctionReferenceNode()
  ]}
]
hierarchy [
 root {loc=<0 0 0>, rot=<0 0 0 1>}
 pelvis {parent=root, loc=<0 -0.5301 98.0571>, rot=<0.0127 0.707 -0.0127 -0.707>}
 spine_01 {parent=pelvis, loc=<6.1208 -0.9328 0>, rot=<0 0 0.034 -0.9994>}
 spine_02 {parent=spine_01, loc=<6.5997 -0.2095 0>, rot=<0 0 0.0238 -0.9997>}
 spine_03 {parent=spine_02, loc=<6.4914 -0.1237 0>, rot=<0 0 0.0149 -0.9999>}
 spine_04 {parent=spine_03, loc=<6.9598 0.0794 0>, rot=<0 0 -0.0132 -0.9999>}
 spine_05 {parent=spine_04, loc=<6.9987 0 0>, rot=<0 0 -0.0015 -1>}
 neck_01 {parent=spine_05, loc=<11.3138 0.1529 0>, rot=<0 0 -0.2003 -0.9797>}
 neck_02 {parent=neck_01, loc=<4.5627 0 0>, rot=<0 0 0.0504 -0.9987>}
 head {parent=neck_02, loc=<5.3914 0 0>, rot=<0 0 0.115 -0.9934>}
 clavicle_l {parent=spine_05, loc=<5.4831 -0.1053 -3.2946>, rot=<0.0465 -0.7289 -0.009 -0.6829>}
 upperarm_l {parent=clavicle_l, loc=<9.8658 -1.2136 -0.4105>, rot=<0.0164 -0.3245 -0.0269 -0.9454>}
 lowerarm_l {parent=upperarm_l, loc=<26.465 -0.8345 0.259>, rot=<0.0036 0.0161 -0.2298 -0.9731>}
 hand_l {parent=lowerarm_l, loc=<22.7355 -0.3203 -0.6105>, rot=<-0.6784 -0.0223 0.0735 -0.7307>}
 thumb_01_l {parent=hand_l, loc=<1.3118 -0.3191 -2.459>, rot=<0.6546 -0.4042 0.0664 -0.6354>}
 thumb_02_l {parent=thumb_01_l, loc=<5.9336 -0.0448 0.0401>, rot=<-0.0037 0.0213 0.298 -0.9543>}
 thumb_03_l {parent=thumb_02_l, loc=<2.2223 0.3682 -0.0907>, rot=<0.0118 -0.013 0.0066 -0.9998>}
 index_metacarpal_l {parent=hand_l, loc=<1.6535 0.0183 -2.96>, rot=<0.0158 -0.0124 -0.0173 -0.9996>}
 index_01_l {parent=index_metacarpal_l, loc=<8.3008 0.0357 -0.0524>, rot=<-0.0044 0.0485 0.0924 -0.9945>}
 index_02_l {parent=index_01_l, loc=<3.3587 -0.3004 0.0452>, rot=<-0.002 0.0027 0.3225 -0.9466>}
 index_03_l {parent=index_02_l, loc=<2.5578 -0.123 -0.0252>, rot=<-0.002 0.0067 0.2558 -0.9667>}
 middle_metacarpal_l {parent=hand_l, loc=<2.1856 -0.1725 -0.5984>, rot=<0.006 0.0168 -0.0329 -0.9993>}
 middle_01_l {parent=middle_metacarpal_l, loc=<7.3275 -0.2447 0.0129>, rot=<-0.0208 0.0042 0.0858 -0.9961>}
 middle_02_l {parent=middle_01_l, loc=<4.0123 0.1401 0.1175>, rot=<-0.0218 0.0258 0.3512 -0.9357>}
 middle_03_l {parent=middle_02_l, loc=<2.6077 0.0534 0.1373>, rot=<0.0082 0.0188 0.1708 -0.9851>}
 ring_metacarpal_l {parent=hand_l, loc=<2.2074 -0.3625 1.3638>, rot=<-0.0284 0.0722 -0.0458 -0.9959>}
 ring_01_l {parent=ring_metacarpal_l, loc=<7.0295 -0.6197 -0.1446>, rot=<-0.0489 -0.0252 0.1892 -0.9804>}
 ring_02_l {parent=ring_01_l, loc=<3.6174 0.4647 -0.1027>, rot=<0 0 0.2806 -0.9598>}
 ring_03_l {parent=ring_02_l, loc=<2.4152 0.114 0.0199>, rot=<0 0 0.1189 -0.9929>}
 pinky_metacarpal_l {parent=hand_l, loc=<1.7616 -0.9753 3.3814>, rot=<-0.0397 0.0814 -0.069 -0.9935>}
 pinky_01_l {parent=pinky_metacarpal_l, loc=<7.1709 -0.4831 -0.1094>, rot=<-0.0654 -0.0035 0.2242 -0.9723>}
 pinky_02_l {parent=pinky_01_l, loc=<2.8274 0.0893 -0.1043>, rot=<0.0037 0.0011 0.2854 -0.9584>}
 pinky_03_l {parent=pinky_02_l, loc=<2.2046 0.0307 -0.0078>, rot=<-0.0002 0.0098 0.1445 -0.9895>}
 deform_wrist_dn_base_l {parent=hand_l, loc=<0 -1.5326 0>, rot=<0 0 0 -1>}
 deform_wrist_dn_l {parent=deform_wrist_dn_base_l, loc=<0 0 0>, rot=<0 0 0 1>}
 deform_wrist_up_base_l {parent=hand_l, loc=<0 2.3009 0>, rot=<0 0 0 -1>}
 deform_wrist_up_l {parent=deform_wrist_up_base_l, loc=<0 0 0>, rot=<0 0 0 1>}
 lowerarm_twist_02_l {parent=lowerarm_l, loc=<10.6707 0 0>, rot=<0 0 0 1>}
 lowerarm_twist_01_l {parent=lowerarm_l, loc=<18.316 0 0>, rot=<0 0 0 1>}
 deform_elbow_in_base_l {parent=lowerarm_l, loc=<0.1834 3.6062 -0.2572>, rot=<0 0 0 -1>}
 deform_elbow_in_l {parent=deform_elbow_in_base_l, loc=<0 0 0>, rot=<0 0 0 1>}
 deform_elbow_out_base_l {parent=lowerarm_l, loc=<0 -4 0>, rot=<0 0 0 1>}
 deform_elbow_out_l {parent=deform_elbow_out_base_l, loc=<0 0 0>, rot=<0 0 0 1>}
 upperarm_twist_01_l {parent=upperarm_l, loc=<7.2135 -0.1134 0>, rot=<0 0 0 1>}
 deform_upperarm_bk_base_l {parent=upperarm_twist_01_l, loc=<-6.4053 -4.3746 0.314>, rot=<0 0 0 -1>}
 deform_upperarm_bk_l {parent=deform_upperarm_bk_base_l, loc=<0 0 0>, rot=<0 0 0 1>}
 deform_upperarm_fr_base_l {parent=upperarm_twist_01_l, loc=<-5.6312 5.092 0.3354>, rot=<0 0 0 -1>}
 deform_upperarm_fr_l {parent=deform_upperarm_fr_base_l, loc=<0 0 0>, rot=<0 0 0 1>}
 deform_upperarm_up_base_l {parent=upperarm_twist_01_l, loc=<-8.891 0.4825 3.9517>, rot=<0 0 0 -1>}
 deform_upperarm_up_l {parent=deform_upperarm_up_base_l, loc=<0 0 0>, rot=<0 0 0 1>}
 upperarm_twist_02_l {parent=upperarm_l, loc=<19.7998 -0.1134 0>, rot=<0 0 0 1>}
 clavicle_r {parent=spine_05, loc=<5.4831 -0.1053 3.2946>, rot=<0.7289 0.0465 -0.6829 0.009>}
 upperarm_r {parent=clavicle_r, loc=<-9.8658 1.2136 0.4105>, rot=<0.0164 -0.3245 -0.0269 -0.9454>}
 lowerarm_r {parent=upperarm_r, loc=<-26.465 0.8345 -0.259>, rot=<0.0036 0.0161 -0.2298 -0.9731>}
 hand_r {parent=lowerarm_r, loc=<-22.7355 0.3203 0.6106>, rot=<-0.6784 -0.0223 0.0735 -0.7307>}
 thumb_01_r {parent=hand_r, loc=<-1.3118 0.3191 2.459>, rot=<0.6546 -0.4042 0.0664 -0.6354>}
 thumb_02_r {parent=thumb_01_r, loc=<-5.9336 0.0448 -0.0401>, rot=<-0.0037 0.0213 0.298 -0.9543>}
 thumb_03_r {parent=thumb_02_r, loc=<-2.2223 -0.3682 0.0907>, rot=<0.0118 -0.013 0.0066 -0.9998>}
 index_metacarpal_r {parent=hand_r, loc=<-1.6535 -0.0183 2.96>, rot=<0.0158 -0.0124 -0.0173 -0.9996>}
 index_01_r {parent=index_metacarpal_r, loc=<-8.3008 -0.0357 0.0524>, rot=<-0.0044 0.0485 0.0924 -0.9945>}
 index_02_r {parent=index_01_r, loc=<-3.3587 0.3004 -0.0452>, rot=<-0.002 0.0027 0.3225 -0.9466>}
 index_03_r {parent=index_02_r, loc=<-2.5578 0.123 0.0252>, rot=<-0.002 0.0067 0.2558 -0.9667>}
 middle_metacarpal_r {parent=hand_r, loc=<-2.1856 0.1725 0.5984>, rot=<0.006 0.0168 -0.0329 -0.9993>}
 middle_01_r {parent=middle_metacarpal_r, loc=<-7.3275 0.2447 -0.0129>, rot=<-0.0208 0.0042 0.0858 -0.9961>}
 middle_02_r {parent=middle_01_r, loc=<-4.0123 -0.1401 -0.1175>, rot=<-0.0218 0.0258 0.3512 -0.9357>}
 middle_03_r {parent=middle_02_r, loc=<-2.6077 -0.0534 -0.1373>, rot=<0.0082 0.0188 0.1708 -0.9851>}
 ring_metacarpal_r {parent=hand_r, loc=<-2.2074 0.3625 -1.3638>, rot=<-0.0284 0.0722 -0.0458 -0.9959>}
 ring_01_r {parent=ring_metacarpal_r, loc=<-7.0295 0.6197 0.1446>, rot=<-0.0489 -0.0252 0.1892 -0.9804>}
 ring_02_r {parent=ring_01_r, loc=<-3.6174 -0.4647 0.1027>, rot=<0 0 0.2806 -0.9598>}
 ring_03_r {parent=ring_02_r, loc=<-2.4152 -0.114 -0.0199>, rot=<0 0 0.1189 -0.9929>}
 pinky_metacarpal_r {parent=hand_r, loc=<-1.7616 0.9753 -3.3814>, rot=<-0.0397 0.0814 -0.069 -0.9935>}
 pinky_01_r {parent=pinky_metacarpal_r, loc=<-7.1709 0.4831 0.1094>, rot=<-0.0654 -0.0035 0.2242 -0.9723>}
 pinky_02_r {parent=pinky_01_r, loc=<-2.8274 -0.0893 0.1043>, rot=<0.0037 0.0011 0.2854 -0.9584>}
 pinky_03_r {parent=pinky_02_r, loc=<-2.2046 -0.0307 0.0078>, rot=<-0.0002 0.0098 0.1445 -0.9895>}
 deform_wrist_dn_base_r {parent=hand_r, loc=<-0.0001 1.5328 0>, rot=<0 0 0 -1>}
 deform_wrist_dn_r {parent=deform_wrist_dn_base_r, loc=<0 0 0>, rot=<0 0 0 1>}
 deform_wrist_up_base_r {parent=hand_r, loc=<-0.0002 -2.3007 -0.0001>, rot=<0 0 0 -1>}
 deform_wrist_up_r {parent=deform_wrist_up_base_r, loc=<0 0 0>, rot=<0 0 0 1>}
 lowerarm_twist_02_r {parent=lowerarm_r, loc=<-10.6707 0 0>, rot=<0 0 0 -1>}
 lowerarm_twist_01_r {parent=lowerarm_r, loc=<-18.316 0 0>, rot=<0 0 0 -1>}
 deform_elbow_in_base_r {parent=lowerarm_r, loc=<-0.1834 -3.6062 0.2572>, rot=<0 0 0 -1>}
 deform_elbow_in_r {parent=deform_elbow_in_base_r, loc=<0 0 0>, rot=<0 0 0 1>}
 deform_elbow_out_base_r {parent=lowerarm_r, loc=<0 4 -0.0001>, rot=<0 0 0 -1>}
 deform_elbow_out_r {parent=deform_elbow_out_base_r, loc=<0 0 0>, rot=<0 0 0 1>}
 upperarm_twist_01_r {parent=upperarm_r, loc=<-7.2135 0.1134 0>, rot=<0 0 0 -1>}
 deform_upperarm_bk_base_r {parent=upperarm_twist_01_r, loc=<6.4053 4.3746 -0.314>, rot=<0 0 0 -1>}
 deform_upperarm_bk_r {parent=deform_upperarm_bk_base_r, loc=<0 0 0>, rot=<0 0 0 1>}
 deform_upperarm_fr_base_r {parent=upperarm_twist_01_r, loc=<5.6312 -5.092 -0.3354>, rot=<0 0 0 -1>}
 deform_upperarm_fr_r {parent=deform_upperarm_fr_base_r, loc=<0 0 0>, rot=<0 0 0 1>}
 deform_upperarm_up_base_r {parent=upperarm_twist_01_r, loc=<8.891 -0.4825 -3.9517>, rot=<0 0 0 -1>}
 deform_upperarm_up_r {parent=deform_upperarm_up_base_r, loc=<0 0 0>, rot=<0 0 0 1>}
 upperarm_twist_02_r {parent=upperarm_r, loc=<-19.7998 0.1134 0>, rot=<0 0 0 -1>}
 attach_backpack {parent=spine_05, loc=<-0.123 -13.473 0>, rot=<-0.0093 -0.7071 0.0092 -0.707>}
 Backpack_BR {parent=attach_backpack, type=Null, loc=<0 0 0>, rot=<0 0 0 1>}
 attach_cape {parent=spine_05, loc=<9.017 -11.714 0>, rot=<0.0616 -0.7044 -0.0617 -0.7044>}
 cape {parent=attach_cape, type=Null, loc=<0 1 0.5>, rot=<0 0 0 1>}
 base_hood {parent=spine_05, loc=<7.5581 -5.9177 0>, rot=<0.2683 -0.6542 0.6542 -0.2683>}
 dyn_hood {parent=base_hood, loc=<0 0 0>, rot=<0 0 0 1>}
 deform_pec_aim_base_l {parent=spine_04, loc=<3.2456 9.3309 -1.2237>, rot=<0.0603 -0.4033 0.3318 -0.8507>}
 deform_pec_aim_l {parent=deform_pec_aim_base_l, loc=<0 0 0>, rot=<0 0 0 1>}
 deform_pec_base_l {parent=deform_pec_aim_l, loc=<9.3665 0 0>, rot=<0 0 0 1>}
 deform_pec_l {parent=deform_pec_base_l, loc=<0 0 0>, rot=<0 0 0 1>}
 deform_pec_aim_base_r {parent=spine_04, loc=<3.2452 9.3309 1.2237>, rot=<-0.4033 -0.0603 0.8507 0.3318>}
 deform_pec_aim_r {parent=deform_pec_aim_base_r, loc=<0 0 0>, rot=<0 0 0 1>}
 deform_pec_base_r {parent=deform_pec_aim_r, loc=<-9.3667 -0.0003 -0.0004>, rot=<0 0 0 1>}
 deform_pec_r {parent=deform_pec_base_r, loc=<0 0 0>, rot=<0 0 0 1>}
 reverse_spine_02_offset {parent=spine_02, loc=<0 0 0>, rot=<0 0 0 1>}
 reverse_spine_02 {parent=reverse_spine_02_offset, loc=<0 0 0>, rot=<0 0 0 1>}
 base_jacket_b {parent=reverse_spine_02, loc=<7.0849 -9.7878 0>, rot=<0.0399 -0.9992 0 0>}
 dyn_jacket_b {parent=base_jacket_b, loc=<0 0 0>, rot=<0 0 0 1>}
 dyn_jacket_end_b {parent=dyn_jacket_b, loc=<10.1546 0 0>, rot=<0 0 0 1>}
 base_jacket_l_2 {parent=reverse_spine_02, loc=<6.9312 -7.8656 -10.7247>, rot=<0.0245 -0.9682 -0.2486 -0.0121>}
 dyn_jacket_l_2 {parent=base_jacket_l_2, loc=<0 0 0>, rot=<0 0 0 1>}
 dyn_jacket_end_l_2 {parent=dyn_jacket_l_2, loc=<10.1546 0 0>, rot=<0 0 0 1>}
 base_jacket_l_1 {parent=reverse_spine_02, loc=<6.3893 -0.1322 -13.4198>, rot=<-0.0272 0.6941 0.7194 -0.0025>}
 dyn_jacket_l_1 {parent=base_jacket_l_1, loc=<0 0 0>, rot=<0 0 0 1>}
 dyn_jacket_end_l_1 {parent=dyn_jacket_l_1, loc=<10.1546 0 0>, rot=<0 0 0 1>}
 base_jacket_l_0 {parent=reverse_spine_02, loc=<5.6417 8.2645 -11.1348>, rot=<-0.0271 0.662 0.749 -0.0037>}
 dyn_jacket_l_0 {parent=base_jacket_l_0, loc=<0 0 0>, rot=<0 0 0 1>}
 dyn_jacket_end_l_0 {parent=dyn_jacket_l_0, loc=<10.1546 0 0>, rot=<0 0 0 1>}
 base_jacket_r_2 {parent=reverse_spine_02, loc=<6.931 -7.8656 10.7247>, rot=<0.0245 -0.9682 0.2486 0.0121>}
 dyn_jacket_r_2 {parent=base_jacket_r_2, loc=<0 0 0>, rot=<0 0 0 1>}
 dyn_jacket_end_r_2 {parent=dyn_jacket_r_2, loc=<10.154 0 -0.0001>, rot=<0 0 0 1>}
 base_jacket_r_1 {parent=reverse_spine_02, loc=<6.389 -0.1322 13.4198>, rot=<0.0272 -0.6941 0.7194 -0.0025>}
 dyn_jacket_r_1 {parent=base_jacket_r_1, loc=<0 0 0>, rot=<0 0 0 1>}
 dyn_jacket_end_r_1 {parent=dyn_jacket_r_1, loc=<10.154 0 0>, rot=<0 0 0 1>}
 base_jacket_r_0 {parent=reverse_spine_02, loc=<5.6415 8.2644 11.1348>, rot=<0.0271 -0.662 0.749 -0.0037>}
 dyn_jacket_r_0 {parent=base_jacket_r_0, loc=<0 0 0>, rot=<0 0 0 1>}
 dyn_jacket_end_r_0 {parent=dyn_jacket_r_0, loc=<10.154 0 0>, rot=<0 0 0 1>}
 spine_02_for_cons_offset {parent=spine_01, loc=<6.5997 -0.2095 0>, rot=<0 0 0.0238 -0.9997>}
 spine_02_for_cons {parent=spine_02_for_cons_offset, loc=<0 0 0>, rot=<0 0 0 1>}
 thigh_l {parent=pelvis, loc=<-1.0855 -0.2109 -8.7053>, rot=<0.0009 0.0355 0.0155 -0.9993>}
 calf_l {parent=thigh_l, loc=<-42.5284 -0.002 -0.73>, rot=<0.0006 -0.0161 -0.0391 -0.9991>}
 foot_l {parent=calf_l, loc=<-46.7401 -2.5014 -0.592>, rot=<0.0427 -0.0238 0.0353 -0.9982>}
 ball_l {parent=foot_l, loc=<-3.5658 10.8066 -0.1544>, rot=<0.0001 -0.0001 -0.7063 -0.7079>}
 calf_twist_02_l {parent=calf_l, loc=<-19.1546 -1.7943 -0.2638>, rot=<0.0237 0.0095 -0.0035 -0.9997>}
 calf_twist_01_l {parent=calf_l, loc=<-35.8483 -1.9197 -0.5794>, rot=<0.0237 0.0095 -0.0035 -0.9997>}
 deform_knee_in_base_l {parent=calf_l, loc=<0.8031 -5.0231 0.2458>, rot=<0 0 0 -1>}
 deform_knee_in_l {parent=deform_knee_in_base_l, loc=<0 0 0>, rot=<0 0 0 1>}
 deform_knee_out_base_l {parent=calf_l, loc=<0 4.0001 0.0002>, rot=<0 0 0 -1>}
 deform_knee_out_l {parent=deform_knee_out_base_l, loc=<0 0 0>, rot=<0 0 0 1>}
 thigh_twist_01_l {parent=thigh_l, loc=<-27.952 0 0>, rot=<0.0013 0 0 -1>}
 base_cube_leg_l {parent=thigh_twist_01_l, loc=<-1.9703 -4.7201 -8.3685>, rot=<0.1323 -0.6762 0.1761 -0.703>}
 dyn_cube_leg_l {parent=base_cube_leg_l, loc=<0 0 0>, rot=<0 0 0 1>}
 deform_groin_base_l {parent=thigh_l, loc=<0.4801 5.6776 -0.2486>, rot=<0 0 0 -1>}
 deform_groin_l {parent=deform_groin_base_l, loc=<0 0 0>, rot=<0 0 0 1>}
 base_controller {parent=thigh_l, loc=<-5.5467 6.3061 -6.6869>, rot=<-0.3069 -0.6071 -0.2292 -0.6962>}
 dyn_controller {parent=base_controller, loc=<0 0 0>, rot=<0 0 0 1>}
 base_belt_helper_front {parent=dyn_controller, loc=<0.2605 3.9807 3.2074>, rot=<0.1531 -0.0664 0.8678 0.4681>}
 dyn_belt_helper_front {parent=base_belt_helper_front, loc=<0 0 0>, rot=<0 0 0 1>}
 base_belt_helper_back {parent=dyn_controller, loc=<-0.3438 -4.3607 3.2448>, rot=<0.1198 0.0425 0.9223 -0.3649>}
 dyn_belt_helper_back {parent=base_belt_helper_back, loc=<0 0 0>, rot=<0 0 0 1>}
 base_joystick {parent=dyn_controller, loc=<2.7668 -0.3764 3.4545>, rot=<-0.0088 0.2522 0.0338 -0.9671>}
 dyn_joystick {parent=base_joystick, loc=<0 0 0>, rot=<0 0 0 1>}
 thigh_r {parent=pelvis, loc=<-1.0855 -0.2109 8.7053>, rot=<0.0355 -0.0009 0.9993 0.0155>}
 calf_r {parent=thigh_r, loc=<42.5284 0.002 0.73>, rot=<0.0006 -0.0161 -0.0391 -0.9991>}
 foot_r {parent=calf_r, loc=<46.7401 2.5014 0.592>, rot=<0.0427 -0.0238 0.0353 -0.9982>}
 ball_r {parent=foot_r, loc=<3.5658 -10.8066 0.1544>, rot=<0.0001 -0.0001 -0.7063 -0.7079>}
 calf_twist_02_r {parent=calf_r, loc=<19.1546 1.7943 0.2638>, rot=<0.0237 0.0095 -0.0035 -0.9997>}
 calf_twist_01_r {parent=calf_r, loc=<35.8483 1.9197 0.5794>, rot=<0.0237 0.0095 -0.0035 -0.9997>}
 deform_knee_in_base_r {parent=calf_r, loc=<-0.8031 5.023 -0.2458>, rot=<0 0 0 -1>}
 deform_knee_in_r {parent=deform_knee_in_base_r, loc=<0 0 0>, rot=<0 0 0 1>}
 deform_knee_out_base_r {parent=calf_r, loc=<0 -4.0001 -0.0002>, rot=<0 0 0 -1>}
 deform_knee_out_r {parent=deform_knee_out_base_r, loc=<0 0 0>, rot=<0 0 0 1>}
 thigh_twist_01_r {parent=thigh_r, loc=<27.952 0 0>, rot=<0.0013 0 0 -1>}
 base_cube_leg_r {parent=thigh_twist_01_r, loc=<1.9703 4.7201 8.3684>, rot=<0.1761 0.703 -0.1323 -0.6762>}
 dyn_cube_leg_r {parent=base_cube_leg_r, loc=<0 0 0>, rot=<0 0 0 1>}
 deform_groin_base_r {parent=thigh_r, loc=<-0.4801 -5.6776 0.2486>, rot=<0 0 0 -1>}
 deform_groin_r {parent=deform_groin_base_r, loc=<0 0 0>, rot=<0 0 0 1>}
 deform_glute_base_l {parent=pelvis, loc=<-7.3466 -9.0213 -9.1245>, rot=<0.0172 0.031 -0.4477 -0.8935>}
 deform_glute_l {parent=deform_glute_base_l, loc=<0 0 0>, rot=<0 0 0 1>}
 deform_glute_base_r {parent=pelvis, loc=<-7.3464 -9.0213 9.1245>, rot=<0.031 -0.0172 0.8935 -0.4477>}
 deform_glute_r {parent=deform_glute_base_r, loc=<0 0 0>, rot=<0 0 0 1>}
 base_main_chain {parent=pelvis, loc=<-2.3023 -13.2309 -6.6765>, rot=<0.1509 -0.9832 -0.0891 0.0513>}
 dyn_main_chain {parent=base_main_chain, loc=<0 0 0>, rot=<0 0 0 1>}
 dyn_heart_chain {parent=dyn_main_chain, loc=<1.8606 0.032 -0.2476>, rot=<0.0383 -0.1395 0.0161 -0.9893>}
 dyn_heart_chain_1 {parent=dyn_heart_chain, loc=<1.7893 0 0>, rot=<0 0 0 1>}
 dyn_cube_chain_0 {parent=dyn_main_chain, loc=<1.2561 -0.1028 0.0877>, rot=<-0.002 0.2333 -0.0112 -0.9723>}
 dyn_cube_chain_1 {parent=dyn_cube_chain_0, loc=<2.1622 0 0>, rot=<-0.0012 0.0302 -0.0398 -0.9988>}
 dyn_cube_chain_2 {parent=dyn_cube_chain_1, loc=<1.9356 0 0>, rot=<-0.0001 -0.0307 0.0043 -0.9995>}
 dyn_cube_chain_3 {parent=dyn_cube_chain_2, loc=<1.7909 0 0>, rot=<0 0 0 1>}
 thigh_l_constraint_offset {parent=pelvis, loc=<-1.0855 -0.2109 -8.7053>, rot=<0.0009 0.0355 0.0155 -0.9992>}
 thigh_l_constraint {parent=thigh_l_constraint_offset, loc=<0 0 0>, rot=<0 0 0 1>}
 thigh_l_support_offset {parent=pelvis, loc=<-1.0855 -0.2109 -8.7053>, rot=<0.0009 0.0355 0.0155 -0.9992>}
 thigh_l_support {parent=thigh_l_support_offset, loc=<0 0 0>, rot=<0 0 0 1>}
 spine_01_null {parent=pelvis, type=Null, loc=<6.1208 -0.9328 0>, rot=<0 0 0.034 -0.9994>}
 for_aim_belt_front {parent=root, loc=<-6.7235 16.0571 100.3294>, rot=<0.1106 -0.0691 0.9713 0.1988>}
 crouching {type=Curve, loc=<0 0 0>, rot=<0 0 0 1>}
 crouching sprinting {type=Curve, loc=<0 0 0>, rot=<0 0 0 1>}
]
variables [
 Idle Animation FrontEnd {type=double, default=0}
 crouching {type=bool, default=False}
 inbetween front side {type=double, default=0}
 leg forward {type=double, default=0}
 leg side {type=double, default=0}
 "sliding  or rail" {type=bool, default=False}
 targeting {type=bool, default=False}
]
