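# Physics asset for the hanging chain accessory on the F_MED_BitFight female medium body cosmetic, simulating a main chain with cube and heart charm strands.
# Capsule bodies link from base_main_chain to dyn_main_chain, which branches into a four-link cube chain and a two-link heart chain.
# Each joint allows limited swing with twist limited to zero degrees.
# Unconstrained pelvis and left thigh twist capsules are also present, likely as collision for the chains.
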PhysicsAsset /BRCosmetics/Characters/Player/Female/Medium/Bodies/F_MED_BitFight/Meshes/F_MED_BitFight_Physics_Chains
bodies {
 base_main_chain [{shapeType=Capsule, length=0.505, radius=0.505, shapeName=base_main_chain_capsule, rotation=<rot 0 -90 0>}]
 dyn_cube_chain_0 [{shapeType=Capsule, length=3.5258, radius=1.1742, shapeName=dyn_cube_chain_0_capsule, center=<3 0 0>, rotation=<rot 90 0 0>}]
 dyn_cube_chain_1 [{shapeType=Capsule, length=3.5734, radius=1.1266, shapeName=dyn_cube_chain_1_capsule, center=<3 0 0>, rotation=<rot 90 0 0>}]
 dyn_cube_chain_2 [{shapeType=Capsule, length=3.4, radius=1.3, shapeName=dyn_cube_chain_2_capsule, center=<3 0 0>, rotation=<rot 90 0 0>}]
 dyn_cube_chain_3 [{shapeType=Capsule, length=1.968, radius=2.732, shapeName=dyn_cube_chain_3_capsule, center=<3.5 0 0>}]
 dyn_heart_chain [{shapeType=Capsule, length=3.5948, radius=1.3316, shapeName=dyn_heart_chain_capsule, center=<3 0 0>, rotation=<rot 90 0 0>}]
 dyn_heart_chain_1 [{shapeType=Capsule, length=2.3975, radius=2.1427, shapeName=dyn_heart_chain_1_capsule, center=<3.5 0 0>}]
 dyn_main_chain [{shapeType=Capsule, length=4.0732, radius=1.2089, shapeName=dyn_main_chain_capsule, center=<5 0 0>, rotation=<rot 90 13.5989 0>}]
 pelvis [{shapeType=Capsule, length=7.8687, radius=13.2914, shapeName=pelvis_capsule}]
 thigh_twist_01_l [{
   shapeType=Capsule
   length=47.9508
   radius=10.0602
   shapeName=thigh_l_capsule
   center=<1.5574 -0.4793 0.0896>
   rotation=<rot 86.7223 179.7592 -176.0451>
  }]
}
constraints [
 {bone1Name=dyn_main_chain, bone2Name=base_main_chain, swing1LimitDegrees=20, swing2LimitDegrees=20, twistLimitDegrees=0}
 {bone1Name=dyn_cube_chain_0, bone2Name=dyn_main_chain, swing1LimitDegrees=20, swing2LimitDegrees=20, twistLimitDegrees=0}
 {bone1Name=dyn_cube_chain_1, bone2Name=dyn_cube_chain_0, swing1LimitDegrees=20, swing2LimitDegrees=20, twistLimitDegrees=0}
 {bone1Name=dyn_cube_chain_2, bone2Name=dyn_cube_chain_1, swing1LimitDegrees=20, swing2LimitDegrees=20, twistLimitDegrees=0}
 {bone1Name=dyn_heart_chain, bone2Name=dyn_main_chain, swing1LimitDegrees=20, swing2LimitDegrees=20, twistLimitDegrees=0}
 {bone1Name=dyn_cube_chain_3, bone2Name=dyn_cube_chain_2, swing1LimitDegrees=20, swing2LimitDegrees=20, twistLimitDegrees=0}
 {bone1Name=dyn_heart_chain_1, bone2Name=dyn_heart_chain, swing1LimitDegrees=20, swing2LimitDegrees=20, twistLimitDegrees=0}
]
