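# Physics asset for secondary motion of the jacket and joystick on the F_MED_BitFight female player cosmetic body.
# Seven two-link jacket chains (back, three left, three right) hang from small base capsules, each link limited to one swing axis with twist locked.
# The joystick is a single link with narrower swing on two axes.
# Unconstrained capsules on the controller, pelvis and right thigh provide additional collision shapes.
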PhysicsAsset /BRCosmetics/Characters/Player/Female/Medium/Bodies/F_MED_BitFight/Meshes/F_MED_BitFight_Physics_Jacket
bodies {
 base_controller [
  {shapeType=Capsule, length=10.4288, radius=3.3994, shapeName=base_controller_capsule, center=<0.2589 -3.6063 -0.4321>, rotation=<rot 0 -90 0>}
  {shapeType=Capsule, length=8.9526, radius=3.6939, shapeName=base_controller_capsule, center=<0.9135 3.3078 -0.4321>, rotation=<rot 0 -90 0>}
 ]
 base_jacket_b [{shapeType=Capsule, length=0.505, radius=0.505, shapeName=base_jacket_b_capsule, rotation=<rot 0 -90 0>}]
 base_jacket_l_0 [{shapeType=Capsule, length=0.505, radius=0.505, shapeName=base_jacket_l_0_capsule, rotation=<rot 0 -90 0>}]
 base_jacket_l_1 [{shapeType=Capsule, length=0.505, radius=0.505, shapeName=base_jacket_l_1_capsule, rotation=<rot 0 -90 0>}]
 base_jacket_l_2 [{shapeType=Capsule, length=0.505, radius=0.505, shapeName=base_jacket_l_2_capsule, rotation=<rot 0 -90 0>}]
 base_jacket_r_0 [{shapeType=Capsule, length=0.505, radius=0.505, shapeName=base_jacket_r_0_capsule, rotation=<rot 0 -90 0>}]
 base_jacket_r_1 [{shapeType=Capsule, length=0.505, radius=0.505, shapeName=base_jacket_r_1_capsule, rotation=<rot 0 -90 0>}]
 base_jacket_r_2 [{shapeType=Capsule, length=0.505, radius=0.505, shapeName=base_jacket_r_2_capsule, rotation=<rot 0 -90 0>}]
 base_joystick [{shapeType=Capsule, length=0.505, radius=0.505, shapeName=base_joystick_capsule, rotation=<rot 0 -90 0>}]
 dyn_jacket_b [{shapeType=Capsule, length=4, radius=4, shapeName=dyn_jacket_b_capsule, center=<2 0 0>}]
 dyn_jacket_end_b [{shapeType=Capsule, length=7.4, radius=4, shapeName=dyn_jacket_end_b_capsule, center=<3 -5 0>}]
 dyn_jacket_end_l_0 [{shapeType=Capsule, length=2.7978, radius=1.9433, shapeName=dyn_jacket_end_l_0_capsule, center=<2 -0.0816 -0.7458>, rotation=<rot 90 0 0>}]
 dyn_jacket_end_l_1 [{
   shapeType=Capsule
   length=6.9518
   radius=2.014
   shapeName=dyn_jacket_end_l_1_capsule
   center=<4 -0.6518 -0.1466>
   rotation=<rot -2.6937 0.0061 2.772>
  }]
 dyn_jacket_end_l_2 [{shapeType=Capsule, length=5.9756, radius=4, shapeName=dyn_jacket_end_l_2_capsule, center=<3 -2.8 0>, rotation=<rot 0 0 16>}]
 dyn_jacket_end_r_0 [{shapeType=Capsule, length=4, radius=4, shapeName=dyn_jacket_end_r_0_capsule, center=<3 -3 0>}]
 dyn_jacket_end_r_1 [{shapeType=Capsule, length=4, radius=4, shapeName=dyn_jacket_end_r_1_capsule, center=<3 -3.5 0>}]
 dyn_jacket_end_r_2 [{shapeType=Capsule, length=5.9756, radius=4, shapeName=dyn_jacket_end_r_2_capsule, center=<3 -2.8 0>, rotation=<rot 0 0 -16>}]
 dyn_jacket_l_0 [{
   shapeType=Capsule
   length=5.8412
   radius=1.8759
   shapeName=dyn_jacket_l_0_capsule
   center=<3.0062 0.7015 0.3341>
   rotation=<rot -89.7604 -32.5027 32.9382>
  }]
 dyn_jacket_l_1 [{shapeType=Capsule, length=5.2197, radius=4.0331, shapeName=dyn_jacket_l_1_capsule, center=<2 0 0>, rotation=<rot -2.6227 -0.0004 -0.0004>}]
 dyn_jacket_l_2 [{shapeType=Capsule, length=4, radius=4, shapeName=dyn_jacket_l_2_capsule, center=<2 0 0>}]
 dyn_jacket_r_0 [{shapeType=Capsule, length=4, radius=4, shapeName=dyn_jacket_r_0_capsule, center=<2 0 0>}]
 dyn_jacket_r_1 [{shapeType=Capsule, length=4, radius=4, shapeName=dyn_jacket_r_1_capsule, center=<2 0 0>}]
 dyn_jacket_r_2 [{shapeType=Capsule, length=4, radius=4, shapeName=dyn_jacket_r_2_capsule, center=<2 0 0>}]
 dyn_joystick [{shapeType=Capsule, length=3.4821, radius=2.3772, shapeName=dyn_joystick_capsule, center=<3.5643 0 0>, rotation=<rot 90 0 0>}]
 pelvis [{shapeType=Capsule, length=3.3803, radius=11.0952, shapeName=pelvis_capsule, center=<0 -2.5036 0>}]
 thigh_twist_01_r [{
   shapeType=Capsule
   length=41.1331
   radius=5.8698
   shapeName=thigh_twist_01_r_capsule
   center=<-4.6458 0 5.1523>
   rotation=<rot 82.0036 179.9994 179.9998>
  }]
}
constraints [
 {bone1Name=dyn_jacket_b, bone2Name=base_jacket_b, swing1LimitDegrees=20, swing2LimitDegrees=0, twistLimitDegrees=0}
 {bone1Name=dyn_jacket_end_b, bone2Name=dyn_jacket_b, swing1LimitDegrees=20, swing2LimitDegrees=0, twistLimitDegrees=0}
 {bone1Name=dyn_jacket_l_0, bone2Name=base_jacket_l_0, swing1LimitDegrees=20, swing2LimitDegrees=0, twistLimitDegrees=0}
 {bone1Name=dyn_jacket_end_l_0, bone2Name=dyn_jacket_l_0, swing1LimitDegrees=20, swing2LimitDegrees=0, twistLimitDegrees=0}
 {bone1Name=dyn_jacket_l_1, bone2Name=base_jacket_l_1, swing1LimitDegrees=20, swing2LimitDegrees=0, twistLimitDegrees=0}
 {bone1Name=dyn_jacket_end_l_1, bone2Name=dyn_jacket_l_1, swing1LimitDegrees=20, swing2LimitDegrees=0, twistLimitDegrees=0}
 {bone1Name=dyn_jacket_l_2, bone2Name=base_jacket_l_2, swing1LimitDegrees=20, swing2LimitDegrees=0, twistLimitDegrees=0}
 {bone1Name=dyn_jacket_end_l_2, bone2Name=dyn_jacket_l_2, swing1LimitDegrees=20, swing2LimitDegrees=0, twistLimitDegrees=0}
 {bone1Name=dyn_jacket_r_0, bone2Name=base_jacket_r_0, swing1LimitDegrees=20, swing2LimitDegrees=0, twistLimitDegrees=0}
 {bone1Name=dyn_jacket_end_r_0, bone2Name=dyn_jacket_r_0, swing1LimitDegrees=20, swing2LimitDegrees=0, twistLimitDegrees=0}
 {bone1Name=dyn_jacket_r_1, bone2Name=base_jacket_r_1, swing1LimitDegrees=20, swing2LimitDegrees=0, twistLimitDegrees=0}
 {bone1Name=dyn_jacket_end_r_1, bone2Name=dyn_jacket_r_1, swing1LimitDegrees=20, swing2LimitDegrees=0, twistLimitDegrees=0}
 {bone1Name=dyn_jacket_r_2, bone2Name=base_jacket_r_2, swing1LimitDegrees=20, swing2LimitDegrees=0, twistLimitDegrees=0}
 {bone1Name=dyn_jacket_end_r_2, bone2Name=dyn_jacket_r_2, swing1LimitDegrees=20, swing2LimitDegrees=0, twistLimitDegrees=0}
 {bone1Name=dyn_joystick, bone2Name=base_joystick, swing1LimitDegrees=10, swing2LimitDegrees=10, twistLimitDegrees=0}
]
