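# Animation Blueprint for the F_MED_BitFight female player body cosmetic, adding secondary motion and pose corrections to a pose copied from the attached parent mesh.
# A Control Rig takes crouch, slide/rail, targeting and front-end idle inputs, then feeds additive bone corrections for jacket, belt, joystick and chain.
# RigidBody chain and jacket simulations follow, then jacket tweaks and AnimDynamics for joystick and hood.
# The update event derives movement-state flags, curve-based values and gravity override.
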AnimBlueprint /BRCosmetics/Characters/Player/Female/Medium/Bodies/F_MED_BitFight/Meshes/F_MED_BitFight_AnimBP
target_skeleton=&/F_MED_BitFight_Skeleton
graphs [
 AnimGraph {nodes [
   Output Pose(Result=Use cached pose 'dynamicsCache'()) {layerGroup=None}
   Save cached pose 'inputPose'(Pose=Copy Pose From Mesh() {bUseAttachedParent=true}) {cacheName=inputPose}
   n1 = Control Rig!(
    Source=Use cached pose 'inputPose'()
    Alpha
    crouching:bool=Get IsCrouchingCustom() -> (IsCrouchingCustom:bool)
    "sliding  or rail":bool=Get isSliding or Rail() -> (isSliding or Rail:bool)
    targeting:bool=Get targeting standing() -> (targeting standing:bool)
    Idle Animation FrontEnd:double=Get Idle Animation FrontEnd() -> (Idle Animation FrontEnd:double)
   ) -> (Pose) {
    alphaBoolBlend {blendOption=Linear}
    controlRigAssetReference {blueprintRigClass=&/F_MED_BitFight_CtrlRig.F_MED_BitFight_CtrlRig_C}
   }
   Save cached pose 'controlRigCache'(Pose=@n1) {cacheName=controlRigCache}
   n2 = RigidBody!(ComponentPose=Local To Component(LocalPose=Use cached pose 'modifyTransformsCache'()), Alpha, OverrideWorldGravity:Vector="0,0,0") -> (Pose) {
    bClampLinearTranslationLimitToRefPose=true
    bOverrideWorldGravity=true
    overridePhysicsAsset=&/F_MED_BitFight_Physics_Chains
    simulationSpace=BaseBoneSpace
    alphaBoolBlend {blendOption=Linear}
    baseBoneRef {boneName=pelvis}
    simSpaceSettings {velocityScaleZ=0.8, worldAlpha=0.45}
   }
   n3 = RigidBody(ComponentPose=@n2, OverrideWorldGravity="0,0,0") {
    bClampLinearTranslationLimitToRefPose=true
    bOverrideWorldGravity=true
    overridePhysicsAsset=&/F_MED_BitFight_Physics_Jacket
    simulationSpace=BaseBoneSpace
    alphaBoolBlend {blendOption=Linear}
    baseBoneRef {boneName=spine_01}
    simSpaceSettings {worldAlpha=0.2}
   }
   n4 = "Constraint - dyn_joystick (T)"(ComponentPose=@n3, Alpha=1, ConstraintWeights_0=1) {
    alphaBoolBlend {blendOption=Linear}
    boneToModify {boneName=dyn_joystick}
    constraintSetup [{transformType=Translation, targetBone {boneName=base_joystick}}]
    constraintWeights [1]
   }
   n5 = "Transform (Modify) Bone - Bone: dyn_jacket_l_0"!(ComponentPose=@n4, bAlphaBoolEnabled:bool=false, Translation="0,-1,0", Rotation="-20.314,24.122,-3.986") -> (Pose) {
    alphaInputType=Bool
    rotationMode=BMM_Additive
    rotationSpace=BCS_BoneSpace
    translationMode=BMM_Additive
    translationSpace=BCS_BoneSpace
    alphaBoolBlend {blendInTime=3, blendOption=Linear, blendOutTime=3}
    alphaScaleBiasClamp {bInterpResult=true}
    boneToModify {boneName=dyn_jacket_l_0}
    rotation=<rot -20.314 24.122 -3.986>
    translation {y=-1}
   }
   n6 = "Transform (Modify) Bone - Bone: dyn_jacket_end_l_0"!(ComponentPose=@n5, bAlphaBoolEnabled:bool=false, Translation="0,-2,0", Rotation="11.405,32.052,15.265") -> (Pose) {
    alphaInputType=Bool
    rotationMode=BMM_Additive
    rotationSpace=BCS_BoneSpace
    translationMode=BMM_Additive
    translationSpace=BCS_BoneSpace
    alphaBoolBlend {blendInTime=3, blendOption=Linear, blendOutTime=3}
    alphaScaleBiasClamp {bInterpResult=true}
    boneToModify {boneName=dyn_jacket_end_l_0}
    rotation=<rot 11.405 32.052 15.265>
    translation {y=-2}
   }
   n7 = "Transform (Modify) Bone - Bone: dyn_jacket_l_1"!(ComponentPose=@n6, bAlphaBoolEnabled:bool=false, Rotation="-0.711,15.066,-12.268") -> (Pose) {
    alphaInputType=Bool
    rotationMode=BMM_Additive
    rotationSpace=BCS_BoneSpace
    translationSpace=BCS_BoneSpace
    alphaBoolBlend {blendInTime=3, blendOption=Linear, blendOutTime=3}
    alphaScaleBiasClamp {bInterpResult=true}
    boneToModify {boneName=dyn_jacket_l_1}
    rotation=<rot -0.711 15.066 -12.268>
   }
   n8 = "Transform (Modify) Bone - Bone: dyn_jacket_end_l_1"!(ComponentPose=@n7, bAlphaBoolEnabled:bool=false, Rotation="-4.767,3.206,-6.529") -> (Pose) {
    alphaInputType=Bool
    rotationMode=BMM_Additive
    rotationSpace=BCS_BoneSpace
    translationSpace=BCS_BoneSpace
    alphaBoolBlend {blendInTime=3, blendOption=Linear, blendOutTime=3}
    alphaScaleBiasClamp {bInterpResult=true}
    boneToModify {boneName=dyn_jacket_end_l_1}
    rotation=<rot -4.767 3.206 -6.529>
   }
   n9 = "Transform (Modify) Bone - Bone: dyn_joystick"!(ComponentPose=@n8, Alpha=0, Rotation="0,20,0", Scale="0.75,0.75,0.75") -> (Pose) {
    rotationMode=BMM_Additive
    rotationSpace=BCS_BoneSpace
    scaleMode=BMM_Additive
    scaleSpace=BCS_BoneSpace
    translationSpace=BCS_BoneSpace
    alphaBoolBlend {blendInTime=3, blendOption=Linear, blendOutTime=3}
    boneToModify {boneName=dyn_joystick}
    rotation {yaw=20}
    scale=<0.75 0.75 0.75>
   }
   n10 = Anim Dynamics - Bone: dyn_joystick!(ComponentPose=@n9, Alpha=1) -> (Pose) {
    angularDampingOverride=1
    angularSpringConstant=5000
    bAngularSpring=true
    bOverrideAngularDamping=true
    bUseGravityOverride=true
    gravityScale=0.35
    simulationSpace=RootRelative
    alphaBoolBlend {blendOption=Linear}
    boundBone {boneName=dyn_joystick}
    componentAppliedLinearAccClamp=<3000 3000 3000>
    componentLinearAccScale=<0.5 0.5 0.5>
    componentLinearVelScale=<0.5 0.5 0.5>
    gravityOverride {z=-980}
    physicsBodyDefinitions [{
      boundBone {boneName=dyn_joystick}
      boxExtents=<40 40 40>
      constraintSetup {angularLimitsMax=<0 5 5>, angularLimitsMin=<0 -5 -5>, angularTarget=<1 0 0>}
      localJointOffset=<8 0 0>
     }]
   }
   n11 = Anim Dynamics - Bone: dyn_hood!(ComponentPose=@n10, Alpha=1) -> (Pose) {
    angularDampingOverride=1
    angularSpringConstant=20000
    bAngularSpring=true
    bOverrideAngularDamping=true
    bUseGravityOverride=true
    gravityScale=0.35
    simulationSpace=RootRelative
    alphaBoolBlend {blendOption=Linear}
    boundBone {boneName=dyn_hood}
    componentAppliedLinearAccClamp=<3000 3000 3000>
    componentLinearAccScale=<0.52 0.52 0.52>
    componentLinearVelScale=<0.52 0.52 0.52>
    gravityOverride {z=-980}
    physicsBodyDefinitions [{
      boundBone {boneName=dyn_hood}
      boxExtents=<30 30 30>
      constraintSetup {angularLimitsMax=<0 0 10>, angularLimitsMin=<0 -10 -10>, angularTarget=<1 0 0>}
      localJointOffset=<8 0 0>
     }]
    simSpaceSettings {maxAngularAcceleration=35, maxAngularVelocity=7, simSpaceAngularAlpha=0.35}
   }
   Save cached pose 'dynamicsCache'(Pose=Component To Local(ComponentPose=@n11)) {cacheName=dynamicsCache}
   n12 = "Transform (Modify) Bone - Bone: base_jacket_l_0"!(
    ComponentPose=Local To Component(LocalPose=Use cached pose 'controlRigCache'())
    bAlphaBoolEnabled:bool=True
    Rotation="-9.570,-27.198,-26.141"
   ) -> (Pose) {
    alphaInputType=Bool
    rotationMode=BMM_Additive
    rotationSpace=BCS_BoneSpace
    alphaBoolBlend {blendInTime=1, blendOption=Linear, blendOutTime=1}
    boneToModify {boneName=base_jacket_l_0}
    rotation=<rot -9.57 -27.198 -26.141>
   }
   n13 = "Transform (Modify) Bone - Bone: base_joystick"!(ComponentPose=@n12, bAlphaBoolEnabled:bool=True, Rotation="-29.811,-0.000,-0.000") -> (Pose) {
    alphaInputType=Bool
    rotationMode=BMM_Additive
    rotationSpace=BCS_BoneSpace
    alphaBoolBlend {blendInTime=1, blendOption=Linear, blendOutTime=1}
    boneToModify {boneName=base_joystick}
    rotation {pitch=-29.811}
   }
   n14 = "Transform (Modify) Bone - Bone: base_jacket_r_0"!(ComponentPose=@n13, bAlphaBoolEnabled:bool=True, Rotation="-0.111,-35.509,0.017") -> (Pose) {
    alphaInputType=Bool
    rotationMode=BMM_Additive
    rotationSpace=BCS_BoneSpace
    alphaBoolBlend {blendInTime=1, blendOption=Linear, blendOutTime=1}
    boneToModify {boneName=base_jacket_r_0}
    rotation=<rot -0.111 -35.509 0.017>
   }
   n15 = "Transform (Modify) Bone - Bone: base_belt_helper_front"!(ComponentPose=@n14, bAlphaBoolEnabled:bool=True, Translation="0.068,0.347,-0.100", Rotation="-18.730,-20.267,25.342") -> (Pose) {
    alphaInputType=Bool
    rotationMode=BMM_Additive
    rotationSpace=BCS_BoneSpace
    translationMode=BMM_Additive
    translationSpace=BCS_BoneSpace
    alphaBoolBlend {blendInTime=1, blendOption=Linear, blendOutTime=1}
    boneToModify {boneName=base_belt_helper_front}
    rotation=<rot -18.73 -20.267 25.342>
    translation=<0.068 0.347 -0.1>
   }
   n16 = "Transform (Modify) Bone - Bone: base_belt_helper_back"!(ComponentPose=@n15, Alpha=0, Rotation="20,20,0") -> (Pose) {
    rotationMode=BMM_Additive
    rotationSpace=BCS_BoneSpace
    alphaBoolBlend {blendInTime=1, blendOption=Linear, blendOutTime=1}
    alphaScaleBiasClamp {bInterpResult=true}
    boneToModify {boneName=base_belt_helper_back}
    rotation {pitch=20, yaw=20}
   }
   n17 = "Transform (Modify) Bone - Bone: base_controller"!(ComponentPose=@n16, Alpha=0, Translation="0.376,0.282,-0.575", Rotation="3.973,0.021,-0.009") -> (Pose) {
    rotationMode=BMM_Additive
    rotationSpace=BCS_BoneSpace
    translationMode=BMM_Additive
    translationSpace=BCS_BoneSpace
    alphaBoolBlend {blendInTime=1, blendOption=Linear, blendOutTime=1}
    alphaScaleBiasClamp {bInterpResult=true}
    boneToModify {boneName=base_controller}
    rotation=<rot 3.973 0.021 -0.009>
    translation=<0.376 0.282 -0.575>
   }
   n18 = "Transform (Modify) Bone - Bone: base_jacket_r_2"!(ComponentPose=@n17, Alpha=0, Rotation="0.015,-6.949,-0.043") -> (Pose) {
    rotationMode=BMM_Additive
    rotationSpace=BCS_BoneSpace
    translationSpace=BCS_BoneSpace
    alphaBoolBlend {blendInTime=1, blendOption=Linear, blendOutTime=1}
    alphaScaleBiasClamp {bInterpResult=true}
    boneToModify {boneName=base_jacket_r_2}
    rotation=<rot 0.015 -6.949 -0.043>
   }
   n19 = "Transform (Modify) Bone - Bone: base_jacket_b"!(ComponentPose=@n18, Alpha=0, Rotation="-0.000,-4.565,0.006") -> (Pose) {
    rotationMode=BMM_Additive
    rotationSpace=BCS_BoneSpace
    translationSpace=BCS_BoneSpace
    alphaBoolBlend {blendInTime=1, blendOption=Linear, blendOutTime=1}
    alphaScaleBiasClamp {bInterpResult=true}
    boneToModify {boneName=base_jacket_b}
    rotation {roll=0.006, yaw=-4.565}
   }
   n20 = "Transform (Modify) Bone - Bone: base_jacket_b"!(ComponentPose=@n19, bAlphaBoolEnabled:bool=True, Rotation="0,-16.543,0") -> (Pose) {
    alphaInputType=Bool
    rotationMode=BMM_Additive
    rotationSpace=BCS_BoneSpace
    translationMode=BMM_Additive
    translationSpace=BCS_BoneSpace
    alphaBoolBlend {blendInTime=1, blendOption=Linear, blendOutTime=1}
    boneToModify {boneName=base_jacket_b}
    rotation {yaw=-16.543}
   }
   n21 = "Transform (Modify) Bone - Bone: base_jacket_l_2"!(ComponentPose=@n20, bAlphaBoolEnabled:bool=True, Rotation="0.000,-16.543,0.000") -> (Pose) {
    alphaInputType=Bool
    rotationMode=BMM_Additive
    rotationSpace=BCS_BoneSpace
    translationMode=BMM_Additive
    translationSpace=BCS_BoneSpace
    alphaBoolBlend {blendInTime=1, blendOption=Linear, blendOutTime=1}
    boneToModify {boneName=base_jacket_l_2}
    rotation {pitch=0, yaw=-16.543}
   }
   n22 = "Transform (Modify) Bone - Bone: base_jacket_r_2"!(ComponentPose=@n21, bAlphaBoolEnabled:bool=True, Rotation="0.000,-16.543,0.000") -> (Pose) {
    alphaInputType=Bool
    rotationMode=BMM_Additive
    rotationSpace=BCS_BoneSpace
    translationMode=BMM_Additive
    translationSpace=BCS_BoneSpace
    alphaBoolBlend {blendInTime=1, blendOption=Linear, blendOutTime=1}
    boneToModify {boneName=base_jacket_r_2}
    rotation {pitch=0, yaw=-16.543}
   }
   n23 = "Constraint - reverse_spine_02 (R)"!(ComponentPose=@n22, bAlphaBoolEnabled:bool=True, ConstraintWeights_0=1) -> (Pose) {
    alphaInputType=Bool
    alphaBoolBlend {blendInTime=1, blendOption=Linear, blendOutTime=1}
    boneToModify {boneName=reverse_spine_02}
    constraintSetup [{transformType=Rotation, targetBone {boneName=pelvis}}]
    constraintWeights [1]
   }
   n24 = "Transform (Modify) Bone - Bone: base_jacket_l_0"!(ComponentPose=@n23, bAlphaBoolEnabled=True, Translation:Vector="0,0,0", Rotation="0,-15,-15") -> (Pose) {
    alphaInputType=Bool
    rotationMode=BMM_Additive
    rotationSpace=BCS_BoneSpace
    translationMode=BMM_Additive
    translationSpace=BCS_BoneSpace
    alphaBoolBlend {blendOption=Linear}
    boneToModify {boneName=base_jacket_l_0}
    rotation {roll=-15, yaw=-15}
   }
   n25 = "Transform (Modify) Bone - Bone: base_jacket_l_1"!(ComponentPose=@n24, bAlphaBoolEnabled:bool=True, Translation="0,0,0", Rotation="0,-15,-15") -> (Pose) {
    alphaInputType=Bool
    rotationMode=BMM_Additive
    rotationSpace=BCS_BoneSpace
    translationMode=BMM_Additive
    translationSpace=BCS_BoneSpace
    alphaBoolBlend {blendOption=Linear}
    boneToModify {boneName=base_jacket_l_1}
    rotation {roll=-15, yaw=-15}
   }
   n26 = "Transform (Modify) Bone - Bone: base_joystick"(ComponentPose=@n25, bAlphaBoolEnabled=True, Rotation="-60,40,0") {
    alphaInputType=Bool
    rotationMode=BMM_Additive
    rotationSpace=BCS_BoneSpace
    scaleSpace=BCS_BoneSpace
    translationMode=BMM_Additive
    translationSpace=BCS_BoneSpace
    alphaBoolBlend {blendOption=Linear}
    boneToModify {boneName=base_joystick}
    rotation {pitch=-60, yaw=40}
   }
   n27 = "Transform (Modify) Bone - Bone: base_joystick"(ComponentPose=@n26, bAlphaBoolEnabled=True, Rotation="-26.882,7.926,0") {
    alphaInputType=Bool
    rotationMode=BMM_Additive
    rotationSpace=BCS_BoneSpace
    translationMode=BMM_Additive
    translationSpace=BCS_BoneSpace
    alphaBoolBlend {blendInTime=0.2, blendOption=Linear, blendOutTime=0.2}
    boneToModify {boneName=base_joystick}
    rotation {pitch=-26.882, yaw=7.926}
   }
   n28 = "Transform (Modify) Bone - Bone: base_belt_helper_back"!(ComponentPose=@n27, bAlphaBoolEnabled:bool=True, Rotation="0,15,0") -> (Pose) {
    alphaInputType=Bool
    rotationMode=BMM_Additive
    rotationSpace=BCS_BoneSpace
    translationMode=BMM_Additive
    translationSpace=BCS_BoneSpace
    alphaBoolBlend {blendInTime=0.2, blendOption=Linear, blendOutTime=0.2}
    boneToModify {boneName=base_belt_helper_back}
    rotation {yaw=15}
   }
   n29 = "Transform (Modify) Bone - Bone: base_main_chain"!(ComponentPose=@n28, bAlphaBoolEnabled:bool=True, Rotation="0,-20,0") -> (Pose) {
    alphaInputType=Bool
    rotationMode=BMM_Additive
    rotationSpace=BCS_BoneSpace
    translationMode=BMM_Additive
    translationSpace=BCS_BoneSpace
    alphaBoolBlend {blendInTime=0.2, blendOption=Linear, blendOutTime=0.2}
    boneToModify {boneName=base_main_chain}
    rotation {yaw=-20}
   }
   n30 = "Transform (Modify) Bone - Bone: lowerarm_twist_01_r"!(ComponentPose=@n29, Alpha=1, Translation="0,1,-1") -> (Pose) {
    rotationSpace=BCS_BoneSpace
    scaleSpace=BCS_BoneSpace
    translationMode=BMM_Additive
    translationSpace=BCS_BoneSpace
    alphaBoolBlend {blendOption=Linear}
    boneToModify {boneName=lowerarm_twist_01_r}
    translation {y=1, z=-1}
   }
   Save cached pose 'modifyTransformsCache'(Pose=Component To Local(ComponentPose=@n30)) {cacheName=modifyTransformsCache}
  ]}
 EventGraph {nodes [
   Event BlueprintUpdateAnimation()
   > n1 = Sequence()
   > n2 = Gravity Override State Function DEPRECATED(CharacterPartAnimInstance=Self-Reference(), GameTimeInSeconds=GetGameTimeInSeconds(), GravityOverrideParams=Get Gravity Override Params())
   > Set GravityOverride(GravityOverride=@n2)
   > Set crouching RBAN disabled(crouching RBAN disabled:double=float - float(A=1, B=GetCurveValue(CurveName=crouching))) -> (Output_Get:double)
   n3 = Get bIsCrouching()
   > Set targeting standing(targeting standing:bool=AND Boolean(A=Get bIsTargeting(), B=NOT Boolean(A=@n3))) -> (Output_Get:bool)
   n4 = Get bIsRightArmHidden() -> (bIsRightArmHidden:bool)
   n5 = Get bIsLeftArmHidden() -> (bIsLeftArmHidden:bool)
   > Set isCrouchingPickaxe(isCrouchingPickaxe:bool=AND Boolean!(A=Get bIsCrouching(), B=Get bIsPlayingMeleeAnim(), C:bool=Get bIsMoving2D(), D:bool=NOT Boolean(A=@n5), E:bool=NOT Boolean(A=@n4)) -> (ReturnValue)) -> (Output_Get:bool)
   n6 = Get bIsSliding()
   > Set IsCrouchingCustom(IsCrouchingCustom:bool=AND Boolean!(A=Get bIsCrouching(), B=NOT Boolean(A=@n6)) -> (ReturnValue)) -> (Output_Get:bool)
   > Set isSliding or Rail(isSliding or Rail:bool=OR Boolean(A=Get bIsSliding(), B=Get bIsOnGrindRail())) -> (Output_Get:bool)
   n7 = Get bIsCrouchSprinting()
   > Set is Use Pickaxe Gauntlet(is Use Pickaxe Gauntlet:bool=AND Boolean!(A=Get bIsLeftArmHidden(), B=Get bIsRightArmHidden(), C:bool=Get bIsCrouching(), D:bool=Get bIsPlayingMeleeAnim(), E:bool=NOT Boolean(A=@n7)) -> (ReturnValue)) -> (Output_Get:bool)
   > n8 = Set crouching sprinting(crouching sprinting:double=float - float(A=1, B=GetCurveValue(CurveName=crouching sprinting))) -> (Output_Get:double)
   > "Set crouch sprinting turn off "("crouch sprinting turn off ":double=MapRangeClamped(Value=@n8, InRangeA=0.2, InRangeB=0)) -> (Output_Get:double)
   n9 = NOT Boolean(A=Get bIsSkyDiving())
   n10 = Get bIsJumping()
   > Set jumping and falling(jumping and falling:bool=OR Boolean(A=AND Boolean!(A=@n9, B=@n10) -> (ReturnValue), B=Get bIsFalling())) -> (Output_Get:bool)
   n11 = Get bIsCrouching()
   n12 = Get bIsTargeting()
   n13 = Get bIsMoving2D()
   > n14 = Branch(Condition=OR Boolean(A=AND Boolean(A=@n11, B=@n13), B=AND Boolean(A=@n11, B=@n12)))
   > "Set fix jacket post dynamics "("fix jacket post dynamics ":bool=false) -> (Output_Get:bool)
   n15 = Get bIsPlayingMeleeAnim()
   "Set fix jacket post dynamics "(execute=@n14.else, "fix jacket post dynamics "=AND Boolean!(A=@n11, B=NOT Boolean(A=@n15), C:bool=Get bRigidBodyEnabled()) -> (ReturnValue))
   n16 = GetAnimInstance(self=GetOwnerSkeletalMeshComponent())
   Is Valid(exec=@n1.then_1, InputObject=@n16)
   > Set Idle Animation FrontEnd(Idle Animation FrontEnd:double=float * float!(
     A=GetCurveValue(self=@n16, CurveName=DisablePelvisAdjustment)
     B:bool=Montage_IsPlaying(self=@n16, Montage=/BRCosmetics/Animation/Game/MainPlayer/Menu/BR/Female_LobbyIdle_HandsDownWide_Montage)
     C:bool=Get bIsInFrontEnd()
    ) -> (ReturnValue) {is=title K2Node_PromotableOperator}) -> (Output_Get:double)
  ]}
]
variables [
 "crouch sprinting turn off " {type=number}
 crouching RBAN disabled {type=number}
 crouching sprinting {type=number}
 "fix jacket post dynamics " {type=boolean, default=true}
 Gravity Override Params {
  type=GravityOverrideParamsStruct
  default {
   skydive_FloatingDown {gravityOverride=<0 -200 -300>, windAmplitude=<200 -500 200>, windFrequency=<1.5 7 4>, windOffset=<0 -300 200>}
   skydive_GoingUp {gravityOverride=<0 -300 800>}
  }
 }
 GravityOverride {type=Vector}
 Idle Animation FrontEnd {type=number}
 is Use Pickaxe Gauntlet {type=boolean}
 IsCrouchingCustom {type=boolean}
 isCrouchingPickaxe {type=boolean}
 isSliding or Rail {type=boolean}
 jumping and falling {type=boolean}
 targeting standing {type=boolean}
]
