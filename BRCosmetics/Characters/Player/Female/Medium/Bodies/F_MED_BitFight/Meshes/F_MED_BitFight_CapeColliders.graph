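# Physics asset for the medium female BitFight body that, by its name, provides collision shapes for cape simulation.
# Capsules cover the pelvis, spine, clavicles, head and feet, with spine_02 to spine_04 each pairing two differently oriented capsules.
# Dynamic chain and jacket-end bones get small unconstrained capsules, while arm, thigh, calf and neck bodies have no shapes.
# Body constraints follow the skeleton with all swing and twist free, so the listed angle limits are inactive.
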PhysicsAsset /BRCosmetics/Characters/Player/Female/Medium/Bodies/F_MED_BitFight/Meshes/F_MED_BitFight_CapeColliders
bodies {
 calf_l []
 calf_r []
 clavicle_l [{shapeType=Capsule, length=14.5245, radius=10.8412, center=<3.9566 -1.0862 -2.1127>, rotation=<rot -77.5566 177.3915 7.7314>}]
 clavicle_r [{shapeType=Capsule, length=14.5245, radius=10.8412, center=<-3.9566 1.0862 2.1127>, rotation=<rot -77.5566 177.3915 -172.2686>}]
 dyn_cube_chain_0 [{
   shapeType=Capsule
   length=3.4207
   radius=2.1353
   shapeName=dyn_cube_chain_0_capsule
   center=<1.2415 0.0467 -0.0079>
   rotation=<rot 0 94.3506 -89.9566>
  }]
 dyn_cube_chain_1 [{
   shapeType=Capsule
   length=4.0778
   radius=1.8022
   shapeName=dyn_cube_chain_1_capsule
   center=<1.2732 -0.0189 -0.0079>
   rotation=<rot 0 89.396 -87.0732>
  }]
 dyn_cube_chain_2 [{
   shapeType=Capsule
   length=2.4517
   radius=5.5968
   shapeName=dyn_cube_chain_2_capsule
   center=<5.3159 -0.0965 0.0491>
   rotation=<rot 0 -91.35 -88.3759>
  }]
 dyn_heart_chain [{
   shapeType=Capsule
   length=3.5941
   radius=3.3266
   shapeName=dyn_heart_chain_capsule
   center=<3.4536 0.0893 -0.1424>
   rotation=<rot 0 -88.4734 -87.2138>
  }]
 dyn_jacket_end_b [{shapeType=Capsule, length=10.5677, radius=3.8186, shapeName=dyn_jacket_end_b_capsule}]
 dyn_jacket_end_l_0 [{shapeType=Capsule, length=9.5924, radius=4.2377, shapeName=dyn_jacket_end_l_0_capsule}]
 dyn_jacket_end_l_1 [{shapeType=Capsule, length=9.9184, radius=4.1645, shapeName=dyn_jacket_end_l_1_capsule}]
 dyn_jacket_end_l_2 [{shapeType=Capsule, length=9.8487, radius=4.1013, shapeName=dyn_jacket_end_l_2_capsule}]
 dyn_jacket_end_r_0 [{shapeType=Capsule, length=10.1674, radius=3.8125, shapeName=dyn_jacket_end_r_0_capsule}]
 dyn_jacket_end_r_1 [{shapeType=Capsule, length=9.2809, radius=3.9902, shapeName=dyn_jacket_end_r_1_capsule}]
 dyn_jacket_end_r_2 [{shapeType=Capsule, length=9.9529, radius=4.1701, shapeName=dyn_jacket_end_r_2_capsule}]
 dyn_main_chain [{
   shapeType=Capsule
   length=2.3618
   radius=2.4878
   shapeName=dyn_main_chain_capsule
   center=<1.1909 0.0316 0.0042>
   rotation=<rot 0 91.0152 -89.1146>
  }]
 foot_l [{shapeType=Capsule, length=10.758, radius=12.0012, center=<-4.1811 6.0027 -0.338>, rotation=<rot 0.0388 1.4701 -89.3982>}]
 foot_r [{shapeType=Capsule, length=10.758, radius=12.0012, center=<4.1811 -6.0027 0.338>, rotation=<rot 0.0388 1.4701 90.6018>}]
 head [{shapeType=Capsule, length=0, radius=18, center=<9.0697 0.0746 0>, rotation=<rot 0 -90 0>}]
 lowerarm_l []
 lowerarm_r []
 neck_02 []
 pelvis [{shapeType=Capsule, length=16.545, radius=12.5773, center=<-2.4471 -2.5289 -0.1981>}]
 spine_01 [{shapeType=Capsule, length=9.7365, radius=15.0149, center=<-8.9021 -0.8973 0>}]
 spine_02 [
  {shapeType=Capsule, length=9.5213, radius=12.2225, center=<-2.6885 -1.2641 0>}
  {shapeType=Capsule, length=20.6875, radius=9.3125, center=<-1.8317 1.9362 0>, rotation=<rot 90 0 0>}
 ]
 spine_03 [
  {shapeType=Capsule, length=8.4037, radius=12.769, center=<0.0574 0.4416 0>}
  {shapeType=Capsule, length=20.7188, radius=9.2812, center=<-1.2163 0.549 0>, rotation=<rot 90 180 -175.0732>}
 ]
 spine_04 [
  {shapeType=Capsule, length=12.1673, radius=13.6273, center=<5.6234 -0.6779 0>}
  {shapeType=Capsule, length=10.0312, radius=9.2812, center=<-1.2326 1.308 0>, rotation=<rot 90 0 0>}
 ]
 spine_05 [{shapeType=Capsule, length=9.678, radius=15.9001, center=<0.6969 1.91 0>}]
 thigh_l []
 thigh_r []
 upperarm_l []
 upperarm_r []
}
constraints [
 {
  bone1Name=spine_01
  bone2Name=pelvis
  swing1LimitDegrees=45
  swing1Motion=Free
  swing2LimitDegrees=45
  swing2Motion=Free
  twistLimitDegrees=45
  twistMotion=Free
 }
 {
  bone1Name=spine_02
  bone2Name=spine_01
  swing1LimitDegrees=45
  swing1Motion=Free
  swing2LimitDegrees=45
  swing2Motion=Free
  twistLimitDegrees=45
  twistMotion=Free
 }
 {
  bone1Name=spine_03
  bone2Name=spine_02
  swing1LimitDegrees=45
  swing1Motion=Free
  swing2LimitDegrees=45
  swing2Motion=Free
  twistLimitDegrees=45
  twistMotion=Free
 }
 {
  bone1Name=spine_04
  bone2Name=spine_03
  swing1LimitDegrees=45
  swing1Motion=Free
  swing2LimitDegrees=45
  swing2Motion=Free
  twistLimitDegrees=45
  twistMotion=Free
 }
 {
  bone1Name=spine_05
  bone2Name=spine_04
  swing1LimitDegrees=45
  swing1Motion=Free
  swing2LimitDegrees=45
  swing2Motion=Free
  twistLimitDegrees=45
  twistMotion=Free
 }
 {
  bone1Name=clavicle_l
  bone2Name=spine_05
  swing1LimitDegrees=45
  swing1Motion=Free
  swing2LimitDegrees=45
  swing2Motion=Free
  twistLimitDegrees=45
  twistMotion=Free
 }
 {
  bone1Name=upperarm_l
  bone2Name=clavicle_l
  swing1LimitDegrees=45
  swing1Motion=Free
  swing2LimitDegrees=45
  swing2Motion=Free
  twistLimitDegrees=45
  twistMotion=Free
 }
 {
  bone1Name=lowerarm_l
  bone2Name=upperarm_l
  swing1LimitDegrees=45
  swing1Motion=Free
  swing2LimitDegrees=45
  swing2Motion=Free
  twistLimitDegrees=45
  twistMotion=Free
 }
 {
  bone1Name=neck_02
  bone2Name=spine_05
  swing1LimitDegrees=45
  swing1Motion=Free
  swing2LimitDegrees=45
  swing2Motion=Free
  twistLimitDegrees=45
  twistMotion=Free
 }
 {
  bone1Name=head
  bone2Name=neck_02
  swing1LimitDegrees=45
  swing1Motion=Free
  swing2LimitDegrees=45
  swing2Motion=Free
  twistLimitDegrees=45
  twistMotion=Free
 }
 {
  bone1Name=thigh_l
  bone2Name=pelvis
  swing1LimitDegrees=45
  swing1Motion=Free
  swing2LimitDegrees=45
  swing2Motion=Free
  twistLimitDegrees=45
  twistMotion=Free
 }
 {
  bone1Name=foot_l
  bone2Name=calf_l
  swing1LimitDegrees=45
  swing1Motion=Free
  swing2LimitDegrees=45
  swing2Motion=Free
  twistLimitDegrees=45
  twistMotion=Free
 }
 {
  bone1Name=lowerarm_r
  bone2Name=upperarm_r
  swing1LimitDegrees=45
  swing1Motion=Free
  swing2LimitDegrees=45
  swing2Motion=Free
  twistLimitDegrees=45
  twistMotion=Free
 }
 {
  bone1Name=upperarm_r
  bone2Name=clavicle_r
  swing1LimitDegrees=45
  swing1Motion=Free
  swing2LimitDegrees=45
  swing2Motion=Free
  twistLimitDegrees=45
  twistMotion=Free
 }
 {
  bone1Name=clavicle_r
  bone2Name=spine_05
  swing1LimitDegrees=45
  swing1Motion=Free
  swing2LimitDegrees=45
  swing2Motion=Free
  twistLimitDegrees=45
  twistMotion=Free
 }
 {
  bone1Name=thigh_r
  bone2Name=pelvis
  swing1LimitDegrees=45
  swing1Motion=Free
  swing2LimitDegrees=45
  swing2Motion=Free
  twistLimitDegrees=45
  twistMotion=Free
 }
 {
  bone1Name=foot_r
  bone2Name=calf_r
  swing1LimitDegrees=45
  swing1Motion=Free
  swing2LimitDegrees=45
  swing2Motion=Free
  twistLimitDegrees=45
  twistMotion=Free
 }
 {
  bone1Name=calf_r
  bone2Name=thigh_r
  swing1LimitDegrees=45
  swing1Motion=Free
  swing2LimitDegrees=45
  swing2Motion=Free
  twistLimitDegrees=45
  twistMotion=Free
 }
 {
  bone1Name=calf_l
  bone2Name=thigh_l
  swing1LimitDegrees=45
  swing1Motion=Free
  swing2LimitDegrees=45
  swing2Motion=Free
  twistLimitDegrees=45
  twistMotion=Free
 }
]
