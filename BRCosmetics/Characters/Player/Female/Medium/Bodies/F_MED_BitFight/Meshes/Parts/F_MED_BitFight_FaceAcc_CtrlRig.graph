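# Control Rig for the F_MED_BitFight face accessory that offsets the base_hair_0 bone as the head tilts, apparently upward.
# A DeltaRotation function turns head rotation against head_null into clamped weights for additive bone offsets.
# Extra offsets scale those weights by a smoothed, curve-shaped isTargeting blend and a pelvis-height band, likely crouching.
# Inverted weights drive the head up, targeting and crouch-sprint curves; a third DeltaRotation result is unused.
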ControlRigBlueprint /BRCosmetics/Characters/Player/Female/Medium/Bodies/F_MED_BitFight/Meshes/Parts/F_MED_BitFight_FaceAcc_CtrlRig
graphs [
 New Function_ContainedGraph {nodes [
   n1 = RigVMFunctionEntryNode()
   RigVMFunction_MathBoolToInteger = &RigVMFunction_MathBoolToInteger(Value=@n1.onlyTwist)
   RigVMFunction_MathQuaternionSwingTwist = &RigVMFunction_MathQuaternionSwingTwist(
    twistAxis=@n1.twistAxis
    Input=&RigUnit_GetRelativeTransformForItem(
     bChildInitial=False
     bParentInitial=False
     child=(Type=Bone,Name="None")
     parent=(Type=Bone,Name="None")
     child.Name=@n1.child
     parent.Name=@n1.parent
     parent.Type=@n1.parentType
    ).RelativeTransform.Rotation
   )
   RigVMFunction_MathQuaternionToEuler = &RigVMFunction_MathQuaternionToEuler(
    Value=RigVMDispatchNode(
     Values=((),())
     Index=@RigVMFunction_MathBoolToInteger.Result
     Values.0=@RigVMFunction_MathQuaternionSwingTwist.Swing
     Values.1=@RigVMFunction_MathQuaternionSwingTwist.Twist
    ) {template="DISPATCH_RigVMDispatch_SelectInt32(in Index,in Values,out Result)"}.Result
    RotationOrder=@n1.Rotation Order
   )
   > DISPATCH_RigVMDispatch_Print = RigVMDispatchNode(ScreenColor=(R=1,G=1,B=1,A=1), ScreenDuration=1.5, Enabled=@n1.debugRotation, Value=@RigVMFunction_MathQuaternionToEuler.Result) {template="DISPATCH_RigVMDispatch_Print(in Prefix,in Value,in Enabled,in ScreenDuration,in ScreenColor)"}
   Clamp = &RigVMFunction_MathIntClamp(Value=@n1.primAxis)
   DISPATCH_RigVMDispatch_SelectInt32 = RigVMDispatchNode(
    Values=(0,0,0)
    Index=@Clamp.Result
    Values.0=@RigVMFunction_MathQuaternionToEuler.Result.X
    Values.1=@RigVMFunction_MathQuaternionToEuler.Result.Y
    Values.2=@RigVMFunction_MathQuaternionToEuler.Result.Z
   ) {template="DISPATCH_RigVMDispatch_SelectInt32(in Index,in Values,out Result)"}
   Remap = &RigVMFunction_MathDoubleRemap(
    TargetMaximum=1
    TargetMinimum=0
    bClamp=False
    Value=@DISPATCH_RigVMDispatch_SelectInt32.Result
    SourceMaximum=@n1.max
    SourceMinimum=@n1.min
   )
   Multiply = &RigVMFunction_MathDoubleMul(A=&RigVMFunction_MathDoubleClamp(Value=@Remap.Result).Result, B=@n1.premult)
   > RigVMFunctionReturnNode(outputWeight=@Multiply.Result)
  ]}
 RigVMFunctionLibrary {nodes [DeltaRotation = RigVMCollapseNode(ExecuteContext=(), Rotation Order=ZYX, premult=1, twistAxis=(X=1,Y=0,Z=0))]}
 RigVMModel {nodes [
   RigUnit_BeginExecution = &RigUnit_BeginExecution()
   DeltaRotation = RigVMFunctionReferenceNode(
    Rotation Order=ZXY
    child=head
    debugRotation=false
    max=-40
    min=0
    onlyTwist=false
    parent=head_null
    parentType=Null
    premult=1
    primAxis=2
    twistAxis=(X=1,Y=0,Z=0)
    ExecuteContext=@RigUnit_BeginExecution.ExecutePin
   )
   > DeltaRotation_1 = RigVMFunctionReferenceNode(
    Rotation Order=ZXY
    child=head
    debugRotation=false
    max=-18
    min=0
    onlyTwist=false
    parent=head_null
    parentType=Null
    premult=1
    primAxis=2
    twistAxis=(X=1,Y=0,Z=0)
   )
   > /Script/ControlRig.RigUnit_SetCurveValue(
    Curve=head up
    Value=&RigVMFunction_MathDoubleAbs(Value=&RigVMFunction_MathDoubleSub(B=1, A=@DeltaRotation_1.outputWeight).Result).Result
   ) {is=struct RigVMUnitNode}
   > &RigUnit_ModifyTransforms(
    ItemToModify=((Item=(Type=Bone,Name="base_hair_0"),Transform=(Rotation=(X=0,Y=-0.342,Z=0,W=0.9397),Translation=(X=0,Y=0,Z=0),Scale3D=(X=1,Y=1,Z=1))))
    WeightMinimum=0
    Weight=@DeltaRotation.outputWeight
   )
   Cast = &RigVMFunction_MathBoolToFloat(Value=RigVMVariableNode() {variable=isTargeting}.Value)
   AlphaInterp = &RigVMFunction_AlphaInterp(InterpSpeedDecreasing=6.5, InterpSpeedIncreasing=6.5, Scale=1, bClampResult=true, bMapRange=true, Value=@Cast.Result)
   RigVMFunction_AnimEvalRichCurve = &RigVMFunction_AnimEvalRichCurve(
    Curve="(EditorCurveData=(Keys=((InterpMode=RCIM_Cubic,TangentMode=RCTM_User,TangentWeightMode=RCTWM_WeightedNone,Time=0,Value=0,ArriveTangent=0.5688,ArriveTangentWeight=0,LeaveTangent=1.6324,LeaveTangentWeight=0),(InterpMode=RCIM_Cubic,TangentMode=RCTM_User,Tange ...(truncated)"
    SourceMinimum=0
    TargetMinimum=0
    Value=@AlphaInterp.Result
   )
   n1 = &RigVMFunction_MathDoubleMul(A=@DeltaRotation.outputWeight, B=@RigVMFunction_AnimEvalRichCurve.Result)
   > &RigUnit_ModifyTransforms(
    ItemToModify=((Item=(Type=Bone,Name="base_hair_0"),Transform=(Rotation=(X=-0.13,Y=-0.0864,Z=-0.0114,W=0.9877),Translation=(X=0,Y=0,Z=0),Scale3D=(X=1,Y=1,Z=1))))
    WeightMinimum=0
    Weight=@n1.Result
   )
   n2 = &RigVMFunction_MathDoubleMul(A=@DeltaRotation_1.outputWeight, B=@RigVMFunction_AnimEvalRichCurve.Result)
   > &RigUnit_SetCurveValue(
    Curve=targeting up backpack
    Value=&RigVMFunction_MathDoubleAbs(Value=&RigVMFunction_MathDoubleSub(B=1, A=@n2.Result).Result).Result
   )
   n3 = &RigVMFunction_AnimEvalRichCurve(
    Curve=(EditorCurveData=(Keys=((),(Time=0.5,Value=1),(Time=0.8))))
    SourceMaximum=2
    SourceMinimum=0
    TargetMinimum=0
    Value=&RigVMFunction_MathDoubleRemap(
     SourceMaximum=75
     SourceMinimum=45
     TargetMaximum=1
     TargetMinimum=0
     bClamp=false
     Value=&RigUnit_GetRelativeTransformForItem(Child=(Type=Bone,Name="pelvis"), Parent=(Type=Bone,Name="root"), bChildInitial=False, bParentInitial=False).RelativeTransform.Translation.Z
    ).Result
   )
   Multiply = &RigVMFunction_MathDoubleMul(A=@n1.Result, B=@n3.Result)
   > &RigUnit_ModifyTransforms(
    ItemToModify=((Item=(Type=Bone,Name="base_hair_0"),Transform=(Rotation=(X=0.173,Y=-0.0858,Z=0.0151,W=0.9811),Translation=(X=0.6039,Y=-0.8345,Z=-0.8013),Scale3D=(X=1,Y=1,Z=1))))
    WeightMinimum=0
    Weight=@Multiply.Result
   )
   > DeltaRotation_2 = RigVMFunctionReferenceNode(
    Rotation Order=ZXY
    child=head
    debugRotation=false
    max=-5
    min=0
    onlyTwist=false
    parent=head_null
    parentType=Null
    premult=1
    primAxis=2
    twistAxis=(X=1,Y=0,Z=0)
   )
   n4 = &RigVMFunction_MathDoubleMul(B=@n3.Result, A=@DeltaRotation.outputWeight)
   > &RigUnit_ModifyTransforms(
    ItemToModify=((Item=(Type=Bone,Name="base_hair_0"),Transform=(Rotation=(X=0,Y=-0.2588,Z=0,W=0.9659),Translation=(X=0,Y=0,Z=0),Scale3D=(X=1,Y=1,Z=1))))
    WeightMinimum=0
    Weight=@n4.Result
   )
   > &RigUnit_SetCurveValue(
    Curve=head up crouching sprinting
    Value=&RigVMFunction_MathDoubleAbs(Value=&RigVMFunction_MathDoubleSub(B=1, A=@n4.Result).Result).Result
   )
  ]}
]
hierarchy [
 root {loc=<0 0 0>, rot=<0 0 0 1>}
 pelvis {parent=root, loc=<0 -0.5301 98.0571>, rot=<0.0127 0.707 -0.0127 -0.707>}
 spine_01 {parent=pelvis, loc=<6.1208 -0.9328 0>, rot=<0 0 0.034 -0.9994>}
 spine_02 {parent=spine_01, loc=<6.5997 -0.2095 0>, rot=<0 0 0.0238 -0.9997>}
 spine_03 {parent=spine_02, loc=<6.4914 -0.1237 0>, rot=<0 0 0.0149 -0.9999>}
 spine_04 {parent=spine_03, loc=<6.9598 0.0794 0>, rot=<0 0 -0.0132 -0.9999>}
 spine_05 {parent=spine_04, loc=<6.9987 0 0>, rot=<0 0 -0.0015 -1>}
 neck_01 {parent=spine_05, loc=<11.3138 0.1529 0>, rot=<0 0 -0.2003 -0.9797>}
 neck_02 {parent=neck_01, loc=<4.5627 0 0>, rot=<0 0 0.0504 -0.9987>}
 head {parent=neck_02, loc=<5.3914 0 0>, rot=<0 0 0.115 -0.9934>}
 base_hair_0 {parent=head, loc=<12.584 -5.9321 0.0233>, rot=<0.5424 -0.4536 0.4536 -0.5424>}
 dyn_hair_0 {parent=base_hair_0, loc=<0 0 0>, rot=<0 0 0 1>}
 dyn_hair_1 {parent=dyn_hair_0, loc=<7 0 0>, rot=<0 -0.5709 0 -0.821>}
 dyn_hair_2 {parent=dyn_hair_1, loc=<7 0 0>, rot=<0 -0.2062 0 -0.9785>}
 dyn_hair_3 {parent=dyn_hair_2, loc=<7 0 0>, rot=<0 0.0843 0 -0.9964>}
 dyn_hair_4 {parent=dyn_hair_3, loc=<7 0 0>, rot=<0 0.2776 0 -0.9607>}
 dyn_short_hair_2 {parent=dyn_hair_2, loc=<-0.5653 -1.5376 3.9373>, rot=<0 -0.0168 0 -0.9999>}
 base_tiny_hair_0 {parent=dyn_hair_0, loc=<0.2192 2.601 0.6048>, rot=<0.2178 -0.194 -0.4139 -0.8623>}
 dyn_tiny_hair_0 {parent=base_tiny_hair_0, loc=<0 0 0>, rot=<0 0 0 1>}
 dyn_tiny_hair_1 {parent=dyn_tiny_hair_0, loc=<7.0802 0 0>, rot=<0 -0.4213 0 -0.9069>}
 dyn_tiny_hair_2 {parent=dyn_tiny_hair_1, loc=<7.3362 0 0>, rot=<0.0101 -0.1684 -0.0591 -0.9839>}
 dyn_tiny_hair_3 {parent=dyn_tiny_hair_2, loc=<7.333 0 0>, rot=<-0.1645 0.2529 -0.1058 -0.9475>}
 dyn_short_hair_0 {parent=dyn_hair_0, loc=<2.5682 0 3.245>, rot=<0 0 0 1>}
 dyn_short_hair_1 {parent=dyn_hair_0, loc=<2.5682 0 3.245>, rot=<0 0 0 1>}
 base_front_hair_l_0 {parent=head, loc=<10.9482 7.9846 -2.4119>, rot=<0.2114 -0.8497 -0.1337 -0.4642>}
 dyn_front_hair_l_0 {parent=base_front_hair_l_0, loc=<0 0 0>, rot=<0 0 0 1>}
 dyn_front_hair_l_1 {parent=dyn_front_hair_l_0, loc=<3.1986 0 0>, rot=<0.0217 -0.2502 -0.0838 -0.9643>}
 dyn_front_hair_l_2 {parent=dyn_front_hair_l_1, loc=<3.1844 0 0>, rot=<-0.0005 -0.0756 0.0067 -0.9971>}
 dyn_front_hair_l_3 {parent=dyn_front_hair_l_2, loc=<3.2439 0 0>, rot=<0 0 0 1>}
 base_front_hair_r_0 {parent=head, loc=<10.9482 7.9846 2.4119>, rot=<0.4642 0.1337 0.8497 -0.2114>}
 dyn_front_hair_r_0 {parent=base_front_hair_r_0, loc=<0 0 0>, rot=<0 0 0 1>}
 dyn_front_hair_r_1 {parent=dyn_front_hair_r_0, loc=<3.1984 0.0002 0.0003>, rot=<-0.0217 -0.2502 0.0838 -0.9643>}
 dyn_front_hair_r_2 {parent=dyn_front_hair_r_1, loc=<3.1846 -0.0001 -0.0001>, rot=<0.0005 -0.0756 -0.0067 -0.9971>}
 dyn_front_hair_r_3 {parent=dyn_front_hair_r_2, loc=<3.2439 0 0>, rot=<0 0 0 1>}
 head_null {parent=neck_02, type=Null, loc=<5.3914 0 0>, rot=<0 0 0.115 -0.9934>}
 clavicle_l {parent=spine_05, loc=<5.4831 -0.1053 -3.2946>, rot=<0.0498 -0.7289 -0.0125 -0.6827>}
 upperarm_l {parent=clavicle_l, loc=<9.8658 -1.2136 -0.4105>, rot=<0.0164 -0.3245 -0.0269 -0.9454>}
 upperarm_twist_01_l {parent=upperarm_l, loc=<7.2135 -0.1134 0>, rot=<0 0 0 1>}
 deform_upperarm_bk_base_l {parent=upperarm_twist_01_l, loc=<-6.4053 -4.3746 0.314>, rot=<0 0 0 -1>}
 deform_upperarm_bk_l {parent=deform_upperarm_bk_base_l, loc=<0 0 0>, rot=<0 0 0 1>}
 clavicle_r {parent=spine_05, loc=<5.4831 -0.1053 3.2946>, rot=<0.7289 0.0498 -0.6827 0.0125>}
 for_aim_hair {parent=spine_05, loc=<8.8417 -16.835 0.0233>, rot=<0.5584 -0.4338 0.4338 -0.5584>}
 head up {type=Curve, loc=<0 0 0>, rot=<0 0 0 1>}
 CTRL_expressions_browDownL {type=Curve, loc=<0 0 0>, rot=<0 0 0 1>}
 CTRL_expressions_browDownR {type=Curve, loc=<0 0 0>, rot=<0 0 0 1>}
 CTRL_expressions_browLateralL {type=Curve, loc=<0 0 0>, rot=<0 0 0 1>}
 CTRL_expressions_browLateralR {type=Curve, loc=<0 0 0>, rot=<0 0 0 1>}
 CTRL_expressions_browRaiseInL {type=Curve, loc=<0 0 0>, rot=<0 0 0 1>}
 CTRL_expressions_browRaiseInR {type=Curve, loc=<0 0 0>, rot=<0 0 0 1>}
 CTRL_expressions_browRaiseOuterL {type=Curve, loc=<0 0 0>, rot=<0 0 0 1>}
 CTRL_expressions_browRaiseOuterR {type=Curve, loc=<0 0 0>, rot=<0 0 0 1>}
 CTRL_expressions_browUINL {type=Curve, loc=<0 0 0>, rot=<0 0 0 1>}
 CTRL_expressions_browUINR {type=Curve, loc=<0 0 0>, rot=<0 0 0 1>}
 CTRL_expressions_browUOUTL {type=Curve, loc=<0 0 0>, rot=<0 0 0 1>}
 CTRL_expressions_browUOUTR {type=Curve, loc=<0 0 0>, rot=<0 0 0 1>}
 CTRL_expressions_browDINL {type=Curve, loc=<0 0 0>, rot=<0 0 0 1>}
 CTRL_expressions_browDINR {type=Curve, loc=<0 0 0>, rot=<0 0 0 1>}
 CTRL_expressions_browDOUTL {type=Curve, loc=<0 0 0>, rot=<0 0 0 1>}
 CTRL_expressions_browDOUTR {type=Curve, loc=<0 0 0>, rot=<0 0 0 1>}
 CTRL_expressions_eyeBlinkL {type=Curve, loc=<0 0 0>, rot=<0 0 0 1>}
 CTRL_expressions_eyeBlinkR {type=Curve, loc=<0 0 0>, rot=<0 0 0 1>}
 CTRL_expressions_eyeWidenL {type=Curve, loc=<0 0 0>, rot=<0 0 0 1>}
 CTRL_expressions_eyeWidenR {type=Curve, loc=<0 0 0>, rot=<0 0 0 1>}
 CTRL_expressions_eyeSquintInnerL {type=Curve, loc=<0 0 0>, rot=<0 0 0 1>}
 CTRL_expressions_eyeSquintInnerR {type=Curve, loc=<0 0 0>, rot=<0 0 0 1>}
 CTRL_expressions_eyeCheekRaiseL {type=Curve, loc=<0 0 0>, rot=<0 0 0 1>}
 CTRL_expressions_eyeCheekRaiseR {type=Curve, loc=<0 0 0>, rot=<0 0 0 1>}
 CTRL_expressions_eyeUpperLidUpL {type=Curve, loc=<0 0 0>, rot=<0 0 0 1>}
 CTRL_expressions_eyeUpperLidUpR {type=Curve, loc=<0 0 0>, rot=<0 0 0 1>}
 CTRL_expressions_eyeRelaxL {type=Curve, loc=<0 0 0>, rot=<0 0 0 1>}
 CTRL_expressions_eyeRelaxR {type=Curve, loc=<0 0 0>, rot=<0 0 0 1>}
 CTRL_expressions_eyeLowerLidUpL {type=Curve, loc=<0 0 0>, rot=<0 0 0 1>}
 CTRL_expressions_eyeLowerLidUpR {type=Curve, loc=<0 0 0>, rot=<0 0 0 1>}
 CTRL_expressions_eyeLowerLidDownL {type=Curve, loc=<0 0 0>, rot=<0 0 0 1>}
 CTRL_expressions_eyeLowerLidDownR {type=Curve, loc=<0 0 0>, rot=<0 0 0 1>}
 CTRL_expressions_eyelidRollPosUL {type=Curve, loc=<0 0 0>, rot=<0 0 0 1>}
 CTRL_expressions_eyelidRollPosUR {type=Curve, loc=<0 0 0>, rot=<0 0 0 1>}
 CTRL_expressions_eyelidRollPosDL {type=Curve, loc=<0 0 0>, rot=<0 0 0 1>}
 CTRL_expressions_eyelidRollPosDR {type=Curve, loc=<0 0 0>, rot=<0 0 0 1>}
 CTRL_expressions_eyelidRollNegUL {type=Curve, loc=<0 0 0>, rot=<0 0 0 1>}
 CTRL_expressions_eyelidRollNegUR {type=Curve, loc=<0 0 0>, rot=<0 0 0 1>}
 CTRL_expressions_eyelidRollNegDL {type=Curve, loc=<0 0 0>, rot=<0 0 0 1>}
 CTRL_expressions_eyelidRollNegDR {type=Curve, loc=<0 0 0>, rot=<0 0 0 1>}
 CTRL_expressions_eyeLookUpL {type=Curve, loc=<0 0 0>, rot=<0 0 0 1>}
 CTRL_expressions_eyeLookUpR {type=Curve, loc=<0 0 0>, rot=<0 0 0 1>}
 CTRL_expressions_eyeLookDownL {type=Curve, loc=<0 0 0>, rot=<0 0 0 1>}
 CTRL_expressions_eyeLookDownR {type=Curve, loc=<0 0 0>, rot=<0 0 0 1>}
 CTRL_expressions_eyeLookLeftL {type=Curve, loc=<0 0 0>, rot=<0 0 0 1>}
 CTRL_expressions_eyeLookLeftR {type=Curve, loc=<0 0 0>, rot=<0 0 0 1>}
 CTRL_expressions_eyeLookRightL {type=Curve, loc=<0 0 0>, rot=<0 0 0 1>}
 CTRL_expressions_eyeLookRightR {type=Curve, loc=<0 0 0>, rot=<0 0 0 1>}
 CTRL_expressions_eyeEyelidsReduce {type=Curve, loc=<0 0 0>, rot=<0 0 0 1>}
 CTRL_expressions_noseWrinkleL {type=Curve, loc=<0 0 0>, rot=<0 0 0 1>}
 CTRL_expressions_noseWrinkleR {type=Curve, loc=<0 0 0>, rot=<0 0 0 1>}
 CTRL_expressions_noseWrinkleUpperL {type=Curve, loc=<0 0 0>, rot=<0 0 0 1>}
 CTRL_expressions_noseWrinkleUpperR {type=Curve, loc=<0 0 0>, rot=<0 0 0 1>}
 CTRL_expressions_noseNostrilDepressL {type=Curve, loc=<0 0 0>, rot=<0 0 0 1>}
 CTRL_expressions_noseNostrilDepressR {type=Curve, loc=<0 0 0>, rot=<0 0 0 1>}
 CTRL_expressions_noseNostrilDilateL {type=Curve, loc=<0 0 0>, rot=<0 0 0 1>}
 CTRL_expressions_noseNostrilDilateR {type=Curve, loc=<0 0 0>, rot=<0 0 0 1>}
 CTRL_expressions_noseNostrilCompressL {type=Curve, loc=<0 0 0>, rot=<0 0 0 1>}
 CTRL_expressions_noseNostrilCompressR {type=Curve, loc=<0 0 0>, rot=<0 0 0 1>}
 CTRL_expressions_noseRotateL {type=Curve, loc=<0 0 0>, rot=<0 0 0 1>}
 CTRL_expressions_noseRotateR {type=Curve, loc=<0 0 0>, rot=<0 0 0 1>}
 CTRL_expressions_noseSquash {type=Curve, loc=<0 0 0>, rot=<0 0 0 1>}
 CTRL_expressions_noseStretch {type=Curve, loc=<0 0 0>, rot=<0 0 0 1>}
 CTRL_expressions_mouthCheekSuckL {type=Curve, loc=<0 0 0>, rot=<0 0 0 1>}
 CTRL_expressions_mouthCheekSuckR {type=Curve, loc=<0 0 0>, rot=<0 0 0 1>}
 CTRL_expressions_mouthCheekBlowL {type=Curve, loc=<0 0 0>, rot=<0 0 0 1>}
 CTRL_expressions_mouthCheekBlowR {type=Curve, loc=<0 0 0>, rot=<0 0 0 1>}
 CTRL_expressions_mouthLipsBlowL {type=Curve, loc=<0 0 0>, rot=<0 0 0 1>}
 CTRL_expressions_mouthLipsBlowR {type=Curve, loc=<0 0 0>, rot=<0 0 0 1>}
 CTRL_expressions_mouthLeft {type=Curve, loc=<0 0 0>, rot=<0 0 0 1>}
 CTRL_expressions_mouthRight {type=Curve, loc=<0 0 0>, rot=<0 0 0 1>}
 CTRL_expressions_mouthUp {type=Curve, loc=<0 0 0>, rot=<0 0 0 1>}
 CTRL_expressions_mouthDown {type=Curve, loc=<0 0 0>, rot=<0 0 0 1>}
 CTRL_expressions_mouthUpperLipRaiseL {type=Curve, loc=<0 0 0>, rot=<0 0 0 1>}
 CTRL_expressions_mouthUpperLipRaiseR {type=Curve, loc=<0 0 0>, rot=<0 0 0 1>}
 CTRL_expressions_mouthLowerLipDepressL {type=Curve, loc=<0 0 0>, rot=<0 0 0 1>}
 CTRL_expressions_mouthLowerLipDepressR {type=Curve, loc=<0 0 0>, rot=<0 0 0 1>}
 CTRL_expressions_mouthCornerPullL {type=Curve, loc=<0 0 0>, rot=<0 0 0 1>}
 CTRL_expressions_mouthCornerPullR {type=Curve, loc=<0 0 0>, rot=<0 0 0 1>}
 CTRL_expressions_mouthStretchL {type=Curve, loc=<0 0 0>, rot=<0 0 0 1>}
 CTRL_expressions_mouthStretchR {type=Curve, loc=<0 0 0>, rot=<0 0 0 1>}
 CTRL_expressions_mouthDimpleL {type=Curve, loc=<0 0 0>, rot=<0 0 0 1>}
 CTRL_expressions_mouthDimpleR {type=Curve, loc=<0 0 0>, rot=<0 0 0 1>}
 CTRL_expressions_mouthCornerDepressL {type=Curve, loc=<0 0 0>, rot=<0 0 0 1>}
 CTRL_expressions_mouthCornerDepressR {type=Curve, loc=<0 0 0>, rot=<0 0 0 1>}
 CTRL_expressions_mouthPressUL {type=Curve, loc=<0 0 0>, rot=<0 0 0 1>}
 CTRL_expressions_mouthPressUR {type=Curve, loc=<0 0 0>, rot=<0 0 0 1>}
 CTRL_expressions_mouthPressDL {type=Curve, loc=<0 0 0>, rot=<0 0 0 1>}
 CTRL_expressions_mouthPressDR {type=Curve, loc=<0 0 0>, rot=<0 0 0 1>}
 CTRL_expressions_mouthLipsPurseUL {type=Curve, loc=<0 0 0>, rot=<0 0 0 1>}
 CTRL_expressions_mouthLipsPurseUR {type=Curve, loc=<0 0 0>, rot=<0 0 0 1>}
 CTRL_expressions_mouthLipsPurseDL {type=Curve, loc=<0 0 0>, rot=<0 0 0 1>}
 CTRL_expressions_mouthLipsPurseDR {type=Curve, loc=<0 0 0>, rot=<0 0 0 1>}
 CTRL_expressions_mouthLipsTowardsUL {type=Curve, loc=<0 0 0>, rot=<0 0 0 1>}
 CTRL_expressions_mouthLipsTowardsUR {type=Curve, loc=<0 0 0>, rot=<0 0 0 1>}
 CTRL_expressions_mouthLipsTowardsDL {type=Curve, loc=<0 0 0>, rot=<0 0 0 1>}
 CTRL_expressions_mouthLipsTowardsDR {type=Curve, loc=<0 0 0>, rot=<0 0 0 1>}
 CTRL_expressions_mouthFunnelUL {type=Curve, loc=<0 0 0>, rot=<0 0 0 1>}
 CTRL_expressions_mouthFunnelUR {type=Curve, loc=<0 0 0>, rot=<0 0 0 1>}
 CTRL_expressions_mouthFunnelDL {type=Curve, loc=<0 0 0>, rot=<0 0 0 1>}
 CTRL_expressions_mouthFunnelDR {type=Curve, loc=<0 0 0>, rot=<0 0 0 1>}
 CTRL_expressions_mouthLipsTogetherUL {type=Curve, loc=<0 0 0>, rot=<0 0 0 1>}
 CTRL_expressions_mouthLipsTogetherUR {type=Curve, loc=<0 0 0>, rot=<0 0 0 1>}
 CTRL_expressions_mouthLipsTogetherDL {type=Curve, loc=<0 0 0>, rot=<0 0 0 1>}
 CTRL_expressions_mouthLipsTogetherDR {type=Curve, loc=<0 0 0>, rot=<0 0 0 1>}
 CTRL_expressions_mouthUpperLipBiteL {type=Curve, loc=<0 0 0>, rot=<0 0 0 1>}
 CTRL_expressions_mouthUpperLipBiteR {type=Curve, loc=<0 0 0>, rot=<0 0 0 1>}
 CTRL_expressions_mouthLowerLipBiteL {type=Curve, loc=<0 0 0>, rot=<0 0 0 1>}
 CTRL_expressions_mouthLowerLipBiteR {type=Curve, loc=<0 0 0>, rot=<0 0 0 1>}
 CTRL_expressions_mouthCornerSharpenUL {type=Curve, loc=<0 0 0>, rot=<0 0 0 1>}
 CTRL_expressions_mouthCornerSharpenUR {type=Curve, loc=<0 0 0>, rot=<0 0 0 1>}
 CTRL_expressions_mouthCornerSharpenDL {type=Curve, loc=<0 0 0>, rot=<0 0 0 1>}
 CTRL_expressions_mouthCornerSharpenDR {type=Curve, loc=<0 0 0>, rot=<0 0 0 1>}
 CTRL_expressions_mouthCornerRounderUL {type=Curve, loc=<0 0 0>, rot=<0 0 0 1>}
 CTRL_expressions_mouthCornerRounderUR {type=Curve, loc=<0 0 0>, rot=<0 0 0 1>}
 CTRL_expressions_mouthCornerRounderDL {type=Curve, loc=<0 0 0>, rot=<0 0 0 1>}
 CTRL_expressions_mouthCornerRounderDR {type=Curve, loc=<0 0 0>, rot=<0 0 0 1>}
 CTRL_expressions_mouthUpperLipTowardsTeethL {type=Curve, loc=<0 0 0>, rot=<0 0 0 1>}
 CTRL_expressions_mouthUpperLipTowardsTeethR {type=Curve, loc=<0 0 0>, rot=<0 0 0 1>}
 CTRL_expressions_mouthLowerLipTowardsTeethL {type=Curve, loc=<0 0 0>, rot=<0 0 0 1>}
 CTRL_expressions_mouthLowerLipTowardsTeethR {type=Curve, loc=<0 0 0>, rot=<0 0 0 1>}
 CTRL_expressions_mouthCornerUpL {type=Curve, loc=<0 0 0>, rot=<0 0 0 1>}
 CTRL_expressions_mouthCornerUpR {type=Curve, loc=<0 0 0>, rot=<0 0 0 1>}
 CTRL_expressions_mouthCornerDownL {type=Curve, loc=<0 0 0>, rot=<0 0 0 1>}
 CTRL_expressions_mouthCornerDownR {type=Curve, loc=<0 0 0>, rot=<0 0 0 1>}
 CTRL_expressions_mouthCornerWideL {type=Curve, loc=<0 0 0>, rot=<0 0 0 1>}
 CTRL_expressions_mouthCornerWideR {type=Curve, loc=<0 0 0>, rot=<0 0 0 1>}
 CTRL_expressions_mouthCornerNarrowL {type=Curve, loc=<0 0 0>, rot=<0 0 0 1>}
 CTRL_expressions_mouthCornerNarrowR {type=Curve, loc=<0 0 0>, rot=<0 0 0 1>}
 CTRL_expressions_tongueUp {type=Curve, loc=<0 0 0>, rot=<0 0 0 1>}
 CTRL_expressions_tongueDown {type=Curve, loc=<0 0 0>, rot=<0 0 0 1>}
 CTRL_expressions_tongueLeft {type=Curve, loc=<0 0 0>, rot=<0 0 0 1>}
 CTRL_expressions_tongueRight {type=Curve, loc=<0 0 0>, rot=<0 0 0 1>}
 CTRL_expressions_tongueOut {type=Curve, loc=<0 0 0>, rot=<0 0 0 1>}
 CTRL_expressions_tongueIn {type=Curve, loc=<0 0 0>, rot=<0 0 0 1>}
 CTRL_expressions_tongueRollUp {type=Curve, loc=<0 0 0>, rot=<0 0 0 1>}
 CTRL_expressions_tongueRollDown {type=Curve, loc=<0 0 0>, rot=<0 0 0 1>}
 CTRL_expressions_tongueRollLeft {type=Curve, loc=<0 0 0>, rot=<0 0 0 1>}
 CTRL_expressions_tongueRollRight {type=Curve, loc=<0 0 0>, rot=<0 0 0 1>}
 CTRL_expressions_tongueWide {type=Curve, loc=<0 0 0>, rot=<0 0 0 1>}
 CTRL_expressions_tongueNarrow {type=Curve, loc=<0 0 0>, rot=<0 0 0 1>}
 CTRL_expressions_jawOpen {type=Curve, loc=<0 0 0>, rot=<0 0 0 1>}
 CTRL_expressions_jawLeft {type=Curve, loc=<0 0 0>, rot=<0 0 0 1>}
 CTRL_expressions_jawRight {type=Curve, loc=<0 0 0>, rot=<0 0 0 1>}
 CTRL_expressions_jawFwd {type=Curve, loc=<0 0 0>, rot=<0 0 0 1>}
 CTRL_expressions_jawBack {type=Curve, loc=<0 0 0>, rot=<0 0 0 1>}
 CTRL_expressions_jawChinRaiseDL {type=Curve, loc=<0 0 0>, rot=<0 0 0 1>}
 CTRL_expressions_jawChinRaiseDR {type=Curve, loc=<0 0 0>, rot=<0 0 0 1>}
 CTRL_expressions_jawChinRaiseUL {type=Curve, loc=<0 0 0>, rot=<0 0 0 1>}
 CTRL_expressions_jawChinRaiseUR {type=Curve, loc=<0 0 0>, rot=<0 0 0 1>}
 CTRL_expressions_jawOpenExtreme {type=Curve, loc=<0 0 0>, rot=<0 0 0 1>}
 CTRL_expressions_teethUpU {type=Curve, loc=<0 0 0>, rot=<0 0 0 1>}
 CTRL_expressions_teethUpD {type=Curve, loc=<0 0 0>, rot=<0 0 0 1>}
 CTRL_expressions_teethDownU {type=Curve, loc=<0 0 0>, rot=<0 0 0 1>}
 CTRL_expressions_teethDownD {type=Curve, loc=<0 0 0>, rot=<0 0 0 1>}
 CTRL_expressions_teethLeftU {type=Curve, loc=<0 0 0>, rot=<0 0 0 1>}
 CTRL_expressions_teethLeftD {type=Curve, loc=<0 0 0>, rot=<0 0 0 1>}
 CTRL_expressions_teethRightU {type=Curve, loc=<0 0 0>, rot=<0 0 0 1>}
 CTRL_expressions_teethRightD {type=Curve, loc=<0 0 0>, rot=<0 0 0 1>}
 CTRL_expressions_teethFwdU {type=Curve, loc=<0 0 0>, rot=<0 0 0 1>}
 CTRL_expressions_teethFwdD {type=Curve, loc=<0 0 0>, rot=<0 0 0 1>}
 CTRL_expressions_teethBackU {type=Curve, loc=<0 0 0>, rot=<0 0 0 1>}
 CTRL_expressions_teethBackD {type=Curve, loc=<0 0 0>, rot=<0 0 0 1>}
 CTRL_expressions_teethRotateFwdU {type=Curve, loc=<0 0 0>, rot=<0 0 0 1>}
 CTRL_expressions_teethRotateFwdD {type=Curve, loc=<0 0 0>, rot=<0 0 0 1>}
 CTRL_expressions_teethRotateBackU {type=Curve, loc=<0 0 0>, rot=<0 0 0 1>}
 CTRL_expressions_teethRotateBackD {type=Curve, loc=<0 0 0>, rot=<0 0 0 1>}
 DisableFaceOverride {type=Curve, loc=<0 0 0>, rot=<0 0 0 1>}
 is_3L {type=Curve, loc=<0 0 0>, rot=<0 0 0 1>}
 is_face_bone_anim {type=Curve, loc=<0 0 0>, rot=<0 0 0 1>}
 WM_Smile_L {type=Curve, loc=<0 0 0>, rot=<0 0 0 1>}
 WM_BrowLateral_L {type=Curve, loc=<0 0 0>, rot=<0 0 0 1>}
 WM_BrowLateral_R {type=Curve, loc=<0 0 0>, rot=<0 0 0 1>}
 WM_BrowRaiseInner_L {type=Curve, loc=<0 0 0>, rot=<0 0 0 1>}
 WM_BrowRaiseInner_R {type=Curve, loc=<0 0 0>, rot=<0 0 0 1>}
 WM_BrowRaiseOutter_L {type=Curve, loc=<0 0 0>, rot=<0 0 0 1>}
 WM_BrowRaiseOutter_R {type=Curve, loc=<0 0 0>, rot=<0 0 0 1>}
 WM_CheekRaise_L {type=Curve, loc=<0 0 0>, rot=<0 0 0 1>}
 WM_CheekRaise_R {type=Curve, loc=<0 0 0>, rot=<0 0 0 1>}
 WM_Lips_L {type=Curve, loc=<0 0 0>, rot=<0 0 0 1>}
 WM_Lips_R {type=Curve, loc=<0 0 0>, rot=<0 0 0 1>}
 WM_MouthStretch_L {type=Curve, loc=<0 0 0>, rot=<0 0 0 1>}
 WM_MouthStretch_R {type=Curve, loc=<0 0 0>, rot=<0 0 0 1>}
 WM_NoseWrinkle_L {type=Curve, loc=<0 0 0>, rot=<0 0 0 1>}
 WM_NoseWrinkle_R {type=Curve, loc=<0 0 0>, rot=<0 0 0 1>}
 WM_Smile_R {type=Curve, loc=<0 0 0>, rot=<0 0 0 1>}
 WM_BrowLateral_Offset_L {type=Curve, loc=<0 0 0>, rot=<0 0 0 1>}
 WM_BrowLateral_Offset_R {type=Curve, loc=<0 0 0>, rot=<0 0 0 1>}
 WM_BrowRaiseInner_Offset_L {type=Curve, loc=<0 0 0>, rot=<0 0 0 1>}
 WM_BrowRaiseInner_Offset_R {type=Curve, loc=<0 0 0>, rot=<0 0 0 1>}
 WM_BrowRaiseOutter_Offset_L {type=Curve, loc=<0 0 0>, rot=<0 0 0 1>}
 WM_BrowRaiseOutter_Offset_R {type=Curve, loc=<0 0 0>, rot=<0 0 0 1>}
 WM_CheekRaise_Offset_L {type=Curve, loc=<0 0 0>, rot=<0 0 0 1>}
 WM_CheekRaise_Offset_R {type=Curve, loc=<0 0 0>, rot=<0 0 0 1>}
 WM_Lips_Offset_L {type=Curve, loc=<0 0 0>, rot=<0 0 0 1>}
 WM_Lips_Offset_R {type=Curve, loc=<0 0 0>, rot=<0 0 0 1>}
 WM_MouthStretch_Offset_L {type=Curve, loc=<0 0 0>, rot=<0 0 0 1>}
 WM_MouthStretch_Offset_R {type=Curve, loc=<0 0 0>, rot=<0 0 0 1>}
 WM_Smile_Offset_L {type=Curve, loc=<0 0 0>, rot=<0 0 0 1>}
 WM_Smile_Offset_R {type=Curve, loc=<0 0 0>, rot=<0 0 0 1>}
 WM_NoseWrinkle_Offset_L {type=Curve, loc=<0 0 0>, rot=<0 0 0 1>}
 WM_NoseWrinkle_Offset_R {type=Curve, loc=<0 0 0>, rot=<0 0 0 1>}
 head up backpack {type=Curve, loc=<0 0 0>, rot=<0 0 0 1>}
 targeting up backpack {type=Curve, loc=<0 0 0>, rot=<0 0 0 1>}
 targeting up crouching backpack {type=Curve, loc=<0 0 0>, rot=<0 0 0 1>}
 targeting up crouching {type=Curve, loc=<0 0 0>, rot=<0 0 0 1>}
 head up crouching sprinting {type=Curve, loc=<0 0 0>, rot=<0 0 0 1>}
]
variables [isTargeting {type=bool, default=False}]
