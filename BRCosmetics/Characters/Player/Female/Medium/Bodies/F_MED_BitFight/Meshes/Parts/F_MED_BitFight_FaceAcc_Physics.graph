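# Physics asset for the F_MED_BitFight face accessory mesh, defining simulation bodies for the character's hair.
# Four capsule chains (left and right front locks, main hair, tiny hair) each hang from a small base anchor body.
# Constraints lock twist throughout and limit swing, keeping front locks stiff while main and tiny hair swing wider below their roots.
# Unconstrained head, neck and upper spine capsules provide body volumes, likely for hair collision.
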PhysicsAsset /BRCosmetics/Characters/Player/Female/Medium/Bodies/F_MED_BitFight/Meshes/Parts/F_MED_BitFight_FaceAcc_Physics
bodies {
 base_front_hair_l_0 [{shapeType=Capsule, length=0.505, radius=0.505, shapeName=base_front_hair_l_0_capsule, rotation=<rot 0 -90 0>}]
 base_front_hair_r_0 [{shapeType=Capsule, length=0.505, radius=0.505, shapeName=base_front_hair_r_0_capsule, rotation=<rot 0 -90 0>}]
 base_hair_0 [{shapeType=Capsule, length=0.505, radius=0.505, shapeName=base_hair_0_capsule, rotation=<rot 0 -90 0>}]
 base_tiny_hair_0 [{shapeType=Capsule, length=0.505, radius=0.505, shapeName=base_tiny_hair_0_capsule, rotation=<rot 0 -90 0>}]
 dyn_front_hair_l_0 [{shapeType=Capsule, length=3.2, radius=1.4, shapeName=dyn_front_hair_l_0_capsule, center=<3 0 0.6>, rotation=<rot 0 0 90>}]
 dyn_front_hair_l_1 [{shapeType=Capsule, length=3.2, radius=1.4, shapeName=dyn_front_hair_l_1_capsule, center=<3 0 0.8>, rotation=<rot 0 0 90>}]
 dyn_front_hair_l_2 [{shapeType=Capsule, length=3.2, radius=1.4, shapeName=dyn_front_hair_l_2_capsule, center=<3 0 0.6>, rotation=<rot 0 0 90>}]
 dyn_front_hair_l_3 [{shapeType=Capsule, length=3.2, radius=1.4, shapeName=dyn_front_hair_l_3_capsule, center=<3 0 0>, rotation=<rot 0 0 90>}]
 dyn_front_hair_r_0 [{shapeType=Capsule, length=3.2, radius=1.4, shapeName=dyn_front_hair_r_0_capsule, center=<3 0 0.6>, rotation=<rot 0 0 90>}]
 dyn_front_hair_r_1 [{shapeType=Capsule, length=3.2, radius=1.4, shapeName=dyn_front_hair_r_1_capsule, center=<3 0 0.8>, rotation=<rot 0 0 90>}]
 dyn_front_hair_r_2 [{shapeType=Capsule, length=3.2, radius=1.4, shapeName=dyn_front_hair_r_2_capsule, center=<3 0 0.6>, rotation=<rot 0 0 90>}]
 dyn_front_hair_r_3 [{shapeType=Capsule, length=3.2, radius=1.4, shapeName=dyn_front_hair_r_3_capsule, center=<3 0 0>, rotation=<rot 0 0 90>}]
 dyn_hair_0 [{shapeType=Capsule, length=8, radius=3, shapeName=dyn_hair_0_capsule, center=<4 0 0>, rotation=<rot 90 0 0>}]
 dyn_hair_1 [{shapeType=Capsule, length=8, radius=4, shapeName=dyn_hair_1_capsule, center=<4 -1.5 -1>, rotation=<rot 0 0 90>}]
 dyn_hair_2 [{shapeType=Capsule, length=7.8631, radius=4.1369, shapeName=dyn_hair_2_capsule, center=<4 -2.4 -1>, rotation=<rot 0 0 90>}]
 dyn_hair_3 [
  {shapeType=Capsule, length=10.291, radius=2.1922, shapeName=dyn_hair_3_capsule, center=<1.9537 -4 -6.5238>, rotation=<rot 0 0 90>}
  {shapeType=Capsule, length=11.4558, radius=2.8082, shapeName=dyn_hair_3_capsule, center=<6.5526 -4 -5.1722>, rotation=<rot 0 0 90>}
  {shapeType=Capsule, length=7.8631, radius=4.1369, shapeName=dyn_hair_2_capsule, center=<4 -2.4 -0.0577>, rotation=<rot 0 0 90>}
 ]
 dyn_hair_4 [{shapeType=Capsule, length=8, radius=4, shapeName=dyn_hair_4_capsule, center=<4 0 0>, rotation=<rot 0 0 90>}]
 dyn_tiny_hair_0 [{shapeType=Capsule, length=5.7308, radius=0.7692, shapeName=dyn_tiny_hair_0_capsule, center=<2 0 0>, rotation=<rot 110 0 0>}]
 dyn_tiny_hair_1 [{shapeType=Capsule, length=6.858, radius=0.723, shapeName=dyn_tiny_hair_1_capsule, center=<2 0 0.5>, rotation=<rot 100 0 0>}]
 dyn_tiny_hair_2 [{shapeType=Capsule, length=7.5744, radius=0.7486, shapeName=dyn_tiny_hair_2_capsule, center=<2 0 0>, rotation=<rot 90 0 0>}]
 dyn_tiny_hair_3 [
  {
   shapeType=Capsule
   length=5.926
   radius=1.0669
   shapeName=dyn_tiny_hair_3_capsule
   center=<1.9088 0.0484 -0.0162>
   rotation=<rot 84.5732 -3.1072 0.1691>
  }
  {
   shapeType=Capsule
   length=4.3709
   radius=1.0669
   shapeName=dyn_tiny_hair_3_capsule
   center=<6.6374 -0.0387 2.0127>
   rotation=<rot 35.0361 176.7132 179.9805>
  }
 ]
 head [{shapeType=Capsule, length=6.2051, radius=7.1494, shapeName=head_capsule, center=<2.8316 1.4501 0>, rotation=<rot 90 0 0>}]
 neck_02 [{shapeType=Capsule, length=7.5494, radius=6.0091, shapeName=neck_02_capsule, center=<0 -1.3934 0>, rotation=<rot 90 0 0>}]
 spine_05 [{shapeType=Capsule, length=20.2472, radius=12.8725, shapeName=spine_05_capsule, center=<2.6169 1.0734 0>, rotation=<rot 0 -90 0>}]
}
constraints [
 {bone1Name=dyn_front_hair_l_0, bone2Name=base_front_hair_l_0, swing1LimitDegrees=10, swing2LimitDegrees=10, twistLimitDegrees=0}
 {bone1Name=dyn_front_hair_l_1, bone2Name=dyn_front_hair_l_0, swing1LimitDegrees=5, swing2LimitDegrees=8, twistLimitDegrees=0}
 {bone1Name=dyn_front_hair_l_2, bone2Name=dyn_front_hair_l_1, swing1LimitDegrees=6, swing2LimitDegrees=6, twistLimitDegrees=0}
 {bone1Name=dyn_front_hair_l_3, bone2Name=dyn_front_hair_l_2, swing1LimitDegrees=6, swing2LimitDegrees=4, twistLimitDegrees=0}
 {bone1Name=dyn_front_hair_r_0, bone2Name=base_front_hair_r_0, swing1LimitDegrees=10, swing2LimitDegrees=10, twistLimitDegrees=0}
 {bone1Name=dyn_front_hair_r_1, bone2Name=dyn_front_hair_r_0, swing1LimitDegrees=5, swing2LimitDegrees=8, twistLimitDegrees=0}
 {bone1Name=dyn_front_hair_r_2, bone2Name=dyn_front_hair_r_1, swing1LimitDegrees=6, swing2LimitDegrees=6, twistLimitDegrees=0}
 {bone1Name=dyn_front_hair_r_3, bone2Name=dyn_front_hair_r_2, swing1LimitDegrees=6, swing2LimitDegrees=4, twistLimitDegrees=0}
 {bone1Name=dyn_hair_0, bone2Name=base_hair_0, swing1LimitDegrees=0, swing2LimitDegrees=10, twistLimitDegrees=0}
 {bone1Name=dyn_hair_1, bone2Name=dyn_hair_0, swing1LimitDegrees=10, swing2LimitDegrees=30, twistLimitDegrees=0}
 {bone1Name=dyn_hair_2, bone2Name=dyn_hair_1, swing1LimitDegrees=10, swing2LimitDegrees=30, twistLimitDegrees=0}
 {bone1Name=dyn_hair_3, bone2Name=dyn_hair_2, swing1LimitDegrees=10, swing2LimitDegrees=30, twistLimitDegrees=0}
 {bone1Name=dyn_hair_4, bone2Name=dyn_hair_3, swing1LimitDegrees=10, swing2LimitDegrees=30, twistLimitDegrees=0}
 {bone1Name=dyn_tiny_hair_0, bone2Name=base_tiny_hair_0, swing1LimitDegrees=0, swing2LimitDegrees=10, twistLimitDegrees=0}
 {bone1Name=dyn_tiny_hair_1, bone2Name=dyn_tiny_hair_0, swing1LimitDegrees=20, swing2LimitDegrees=30, twistLimitDegrees=0}
 {bone1Name=dyn_tiny_hair_2, bone2Name=dyn_tiny_hair_1, swing1LimitDegrees=20, swing2LimitDegrees=30, twistLimitDegrees=0}
 {bone1Name=dyn_tiny_hair_3, bone2Name=dyn_tiny_hair_2, swing1LimitDegrees=20, swing2LimitDegrees=30, twistLimitDegrees=0}
]
