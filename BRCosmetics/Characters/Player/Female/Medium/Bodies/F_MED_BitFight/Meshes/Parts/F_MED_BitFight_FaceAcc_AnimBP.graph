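# Post-process style AnimBP for the female BitFight face accessory part, adding procedural and physics motion to hair bones.
# The parent mesh pose passes through a Control Rig into a cache feeding two RigidBody branches with additive bone offsets, one using a backpack physics asset.
# A bool blend picks a branch, then Anim Dynamics springs three short-hair bones.
# The event graph tracks backbling and cape state, head curves, gravity override and a zipline dynamics vector.
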AnimBlueprint /BRCosmetics/Characters/Player/Female/Medium/Bodies/F_MED_BitFight/Meshes/Parts/F_MED_BitFight_FaceAcc_AnimBP
target_skeleton=&/F_MED_BitFight_FaceAcc_Skeleton
graphs [
 AnimGraph {nodes [
   n1 = "Transform (Modify) Bone - Bone: base_hair_0"!(ComponentPose=Local To Component(LocalPose=Use cached pose 'controlRigCache'()), Alpha=1, Rotation="30,0,0", Scale="1,1,1") -> (Pose) {
    rotationMode=BMM_Additive
    rotationSpace=BCS_BoneSpace
    scaleMode=BMM_Additive
    scaleSpace=BCS_BoneSpace
    alphaBoolBlend {blendOption=Linear}
    alphaScaleBiasClamp {bInterpResult=true}
    boneToModify {boneName=base_hair_0}
    rotation {pitch=30}
   }
   n2 = "Transform (Modify) Bone - Bone: base_hair_0"(ComponentPose=@n1, Alpha=1, Rotation="-16.399,-2.558,8.923", Scale="1,1,1") {
    rotationMode=BMM_Additive
    rotationSpace=BCS_BoneSpace
    scaleMode=BMM_Additive
    scaleSpace=BCS_BoneSpace
    translationSpace=BCS_BoneSpace
    alphaBoolBlend {blendOption=Linear}
    alphaScaleBiasClamp {bInterpResult=true}
    boneToModify {boneName=base_hair_0}
    rotation=<rot -16.399 -2.558 8.923>
   }
   n3 = "Transform (Modify) Bone - Bone: base_tiny_hair_0"!(ComponentPose=@n2, Alpha=1, Rotation="15,-1.573,-16.857", Scale="1,1,1") -> (Pose) {
    rotationMode=BMM_Additive
    rotationSpace=BCS_BoneSpace
    scaleMode=BMM_Additive
    scaleSpace=BCS_BoneSpace
    alphaBoolBlend {blendOption=Linear}
    alphaScaleBiasClamp {bInterpResult=true}
    boneToModify {boneName=base_tiny_hair_0}
    rotation=<rot 15 -1.573 -16.857>
   }
   n4 = "Transform (Modify) Bone - Bone: base_tiny_hair_0"(ComponentPose=@n3, Alpha=1, Rotation="3.7,3.7,-10.7", Scale="1,1,1") {
    rotationMode=BMM_Additive
    rotationSpace=BCS_BoneSpace
    scaleMode=BMM_Additive
    scaleSpace=BCS_BoneSpace
    alphaBoolBlend {blendOption=Linear}
    alphaScaleBiasClamp {bInterpResult=true}
    boneToModify {boneName=base_tiny_hair_0}
    rotation=<rot 3.7 3.7 -10.7>
   }
   n5 = RigidBody!(ComponentPose=@n4, Alpha, OverrideWorldGravity:Vector="0,0,0", ExternalForce) -> (Pose) {
    bClampLinearTranslationLimitToRefPose=true
    bOverrideWorldGravity=true
    overridePhysicsAsset=&/F_MED_BitFight_FaceAcc_Physics_Backpack
    simulationSpace=BaseBoneSpace
    alphaBoolBlend {blendOption=Linear}
    baseBoneRef {boneName=neck_02}
    simSpaceSettings {velocityScaleZ=0.9, worldAlpha=0.3, externalLinearDragV {z=0.2}}
   }
   n6 = Component To Local(ComponentPose=@n5)
   n7 = RigidBody(ComponentPose=Local To Component(LocalPose=Use cached pose 'controlRigCache'()), OverrideWorldGravity="0,0,0") {
    bClampLinearTranslationLimitToRefPose=true
    bOverrideWorldGravity=true
    overridePhysicsAsset=&/F_MED_BitFight_FaceAcc_Physics
    simulationSpace=BaseBoneSpace
    alphaBoolBlend {blendOption=Linear}
    baseBoneRef {boneName=neck_02}
    simSpaceSettings {velocityScaleZ=0.9, worldAlpha=0.3, externalLinearDragV {z=0.2}}
   }
   n8 = "Transform (Modify) Bone - Bone: base_hair_0"!(ComponentPose=@n7, bAlphaBoolEnabled:bool=True, Rotation="15,0,0") -> (Pose) {
    alphaInputType=Bool
    rotationMode=BMM_Additive
    rotationSpace=BCS_BoneSpace
    alphaBoolBlend {blendOption=Linear}
    boneToModify {boneName=base_hair_0}
    rotation {pitch=15}
   }
   n9 = Component To Local(ComponentPose=@n8)
   n10 = Anim Dynamics - Bone: dyn_short_hair_0!(
    ComponentPose=Local To Component(LocalPose=Blend Poses by bool(BlendPose_0=@n6, BlendPose_1=@n9) {blendType=Linear})
    Alpha=1
    ComponentLinearAccScale:Vector="0.5,0.5,0.5"
    ComponentLinearVelScale:Vector="0.5,0.5,0.5"
   ) -> (Pose) {
    angularDampingOverride=1
    angularSpringConstant=20000
    bAngularSpring=true
    bOverrideAngularDamping=true
    bUseGravityOverride=true
    gravityScale=0.35
    simulationSpace=RootRelative
    alphaBoolBlend {blendOption=Linear}
    boundBone {boneName=dyn_short_hair_0}
    componentAppliedLinearAccClamp=<3000 3000 3000>
    componentLinearAccScale=<0.5 0.5 0.5>
    componentLinearVelScale=<0.5 0.5 0.5>
    gravityOverride {z=-980}
    physicsBodyDefinitions [{
      boundBone {boneName=dyn_short_hair_0}
      boxExtents=<40 40 40>
      constraintSetup {angularLimitsMax=<0 5 5>, angularLimitsMin=<0 -10 -5>, angularTarget=<1 0 0>}
      localJointOffset=<5 0 0>
     }]
    simSpaceSettings {maxAngularAcceleration=35, maxAngularVelocity=7, simSpaceAngularAlpha=0.35}
   }
   n11 = Anim Dynamics - Bone: dyn_short_hair_1!(ComponentPose=@n10, Alpha=1, ComponentLinearAccScale:Vector="0.5,0.5,0.5", ComponentLinearVelScale:Vector="0.5,0.5,0.5") -> (Pose) {
    angularDampingOverride=1
    angularSpringConstant=15000
    bAngularSpring=true
    bOverrideAngularDamping=true
    bUseGravityOverride=true
    gravityScale=0.35
    simulationSpace=RootRelative
    alphaBoolBlend {blendOption=Linear}
    boundBone {boneName=dyn_short_hair_1}
    componentAppliedLinearAccClamp=<3000 3000 3000>
    componentLinearAccScale=<0.5 0.5 0.5>
    componentLinearVelScale=<0.5 0.5 0.5>
    gravityOverride {z=-980}
    physicsBodyDefinitions [{
      boundBone {boneName=dyn_short_hair_1}
      boxExtents=<40 40 40>
      constraintSetup {angularLimitsMax=<0 2 7>, angularLimitsMin=<0 -10 -7>, angularTarget=<1 0 0>}
      localJointOffset=<5 0 0>
     }]
    simSpaceSettings {maxAngularAcceleration=35, maxAngularVelocity=7, simSpaceAngularAlpha=0.35}
   }
   n12 = Anim Dynamics - Bone: dyn_short_hair_2!(ComponentPose=@n11, Alpha=1, ComponentLinearAccScale:Vector="0.5,0.5,0.5", ComponentLinearVelScale:Vector="0.5,0.5,0.5") -> (Pose) {
    angularDampingOverride=1
    angularSpringConstant=20000
    bAngularSpring=true
    bOverrideAngularDamping=true
    bUseGravityOverride=true
    gravityScale=0.35
    simulationSpace=RootRelative
    alphaBoolBlend {blendOption=Linear}
    boundBone {boneName=dyn_short_hair_2}
    componentAppliedLinearAccClamp=<3000 3000 3000>
    componentLinearAccScale=<0.5 0.5 0.5>
    componentLinearVelScale=<0.5 0.5 0.5>
    gravityOverride {z=-980}
    physicsBodyDefinitions [{
      boundBone {boneName=dyn_short_hair_2}
      boxExtents=<30 30 30>
      constraintSetup {angularLimitsMax=<0 10 10>, angularLimitsMin=<0 -10 -10>, angularTarget=<1 0 0>}
      localJointOffset=<4 0 0>
     }]
    simSpaceSettings {maxAngularAcceleration=35, maxAngularVelocity=7, simSpaceAngularAlpha=0.35}
   }
   Output Pose(Result=Component To Local(ComponentPose=@n12)) {layerGroup=None}
   n13 = Control Rig!(Source=Copy Pose From Mesh() {bUseAttachedParent=true}, Alpha, isTargeting:bool=Get bIsTargeting()) -> (Pose) {
    alphaBoolBlend {blendOption=Linear}
    controlRigAssetReference {blueprintRigClass=&/F_MED_BitFight_FaceAcc_CtrlRig.F_MED_BitFight_FaceAcc_CtrlRig_C}
   }
   Save cached pose 'controlRigCache'(Pose=@n13) {cacheName=controlRigCache}
  ]}
 EventGraph {nodes [
   Event OnCharacterCustomizationLoaded()
   > n1 = Cast To PlayerPawn_Athena(Object=TryGetPawnOwner())
   > Set Body_Mesh(Body_Mesh=GetSkeletalMeshForPartType(self=@n1))
   n2 = CheckForBackbling()
   > Set isWearingBackbling(isWearingBackbling=AND Boolean!(A=@n2, B=NOT Boolean(A=@n2.bIsBackblingACape), C:bool=NOT Boolean(A=@n2.bIsBackblingLong)) -> (ReturnValue))
   > Set Is Wearing Cape(Is Wearing Cape=OR Boolean(A=@n2.bIsBackblingACape, B=@n2.bIsBackblingLong))
   Event BlueprintUpdateAnimation()
   > Set head up(head up:double=float - float(A=1, B=GetCurveValue(CurveName=head up))) -> (Output_Get:double)
   > Set targeting up backpack(targeting up backpack:double=float - float(A=1, B=GetCurveValue(CurveName=targeting up backpack))) -> (Output_Get:double)
   n3 = GetCurveValue(CurveName=head up crouching sprinting)
   > Set head up crouching sprinting(head up crouching sprinting:double=MapRangeClamped(Value=float - float(A=1, B=@n3), InRangeA=0, InRangeB=1, OutRangeA=1, OutRangeB=0.5)) -> (Output_Get:double)
   > n4 = Gravity Override State Function DEPRECATED(CharacterPartAnimInstance=Self-Reference(), GameTimeInSeconds=GetGameTimeInSeconds(), GravityOverrideParams=Get Gravity Override Params())
   > Set GravityOverride(GravityOverride=@n4)
   n5 = Event BlueprintProcessAnimInputEvent()
   n6 = IsNamed(self=@n5.AnimInputEvent, InEventName=aie.Player.Zipline.Entered)
   n7 = OR Boolean(A=Get bIsMoving2D(), B=Get bIsTacticalSprinting())
   n8 = Get bIsJumping()
   > n9 = Branch(Condition=OR Boolean(A=@n6, B=AND Boolean(A=@n6, B=@n8, C=@n7)))
   > Set animDyn on zipline(animDyn on zipline:Vector="0, 0, 0") -> (Output_Get:Vector)
   Branch(execute=@n9.else, Condition=IsNamed(self=@n5.AnimInputEvent, InEventName=aie.Player.Zipline.Exited))
   > Set animDyn on zipline(animDyn on zipline="0.5,0.5,0.5")
  ]}
]
variables [
 animDyn on zipline {type=Vector, default=<0.5 0.5 0.5>}
 Body_Mesh {type=SkeletalMeshComponent}
 Gravity Override Params {
  type=GravityOverrideParamsStruct
  default {
   skydive_DivingDown {gravityOverride=<-1500 0 -900>}
   skydive_FloatingDown {gravityOverride=<0 -700 -300>}
   skydive_GoingUp {gravityOverride=<0 -300 800>}
   skydive_Parachuting {gravityOverride=<0 -1500 -1500>, windAmplitude=<200 -500 -980>, windFrequency=<1.5 7 0>, windOffset=<0 0 -980>}
  }
 }
 GravityOverride {type=Vector}
 head up {type=number}
 head up crouching sprinting {type=number}
 Is Wearing Cape {type=boolean}
 isWearingBackbling {type=boolean}
 targeting up backpack {type=number}
]
